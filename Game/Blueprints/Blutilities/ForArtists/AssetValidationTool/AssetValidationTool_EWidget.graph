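# Editor utility widget for artists that finds and lists invalid assets among level static meshes or selected Content Browser assets.
# Validation results become list items whose clicks open, browse to, show LOD settings for, or select level users of the asset.
# Filter widgets are grouped into category tabs by enum, and their state persists through a SaveGame alongside an editable exclusion set.
# Listed errors can also be assembled into CSV rows for logging.
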EditorUtilityWidgetBlueprint /Game/Blueprints/Blutilities/ForArtists/AssetValidationTool/AssetValidationTool_EWidget
graphs [
 EventGraph {nodes [
   "On Clicked (btnFindBadAssets_InLevel)"!() -> (OutputDelegate, then, widget:User Widget Object Reference, event:Pointer Event Structure) {is=title K2Node_ComponentBoundEvent}
   > clearItems(self:TA Generic 01 List Object Reference=Get TA_Generic01_List() -> (TA_Generic01_List:TA Generic 01 List Object Reference))
   > n1 = findBadMeshesInLevel!(execute, self) -> (then, meshesAndReasons:Map of Static Mesh Object References to AVT Asset Data Struct Structures)
   > addAssetsToList!(execute, self, assetsAndReasons:Map of Object References to AVT Asset Data Struct Structures=@n1) -> (then)
   "On Item Selection Changed (listFilters)"!() -> (
    OutputDelegate
    then
    selectedData:Array of Object References
    selectedWidgets:Array of User Widget Object References
    selectedIndices:Array of Integers
   ) {is=title K2Node_ComponentBoundEvent}
   > handleFilterTypeChanged()
   > SaveGame_Save(self:TA Base Editor Widget Object Reference)
   "On Clicked (btnSelectAllInCB)"!() -> (OutputDelegate, then, widget:User Widget Object Reference, event:Pointer Event Structure) {is=title K2Node_ComponentBoundEvent}
   > n2 = getListData!(execute, self:TA Generic 01 List Object Reference=Get TA_Generic01_List()) -> (then, data:Array of Object References)
   > syncListDataInCB!(execute, self, data:Array of Object References=@n2) -> (then)
   "On Clicked (btnExclude_Selection)"!() -> (OutputDelegate, then, widget:User Widget Object Reference, event:Pointer Event Structure) {is=title K2Node_ComponentBoundEvent}
   > excludeSelectedListItem()
   n3 = "On Item Left Clicked (TA_Generic01_List)"!() -> (OutputDelegate, then, isShiftDown:bool, isCtrlDown:bool, data:Object Reference, item:User Widget Object Reference) {is=title K2Node_ComponentBoundEvent}
   > n4 = Branch(Condition=@n3.isShiftDown)
   > n5 = Cast To AVT_ItemData!(execute, Object=@n3.data) -> (then, CastFailed, AsAVT Item Data:AVT Item Data Object Reference) {is=title K2Node_DynamicCast}
   n6 = Get!(self:AVT Item Data Object Reference=@n5) -> (
    assetErrorData_asset_39_#1:Object Reference
    assetErrorData_errors_35_#2:Array of AVT Error Struct Structures
    assetErrorData_csvLine_42_#3:String
    assetErrorData_assetPackageName_46_#4:String
   )
   > Is Valid(InputObject=@n6)
   > openAssetForEdit!(execute, self, Asset:Object Reference=@n6) -> (then)
   n7 = Branch(execute=@n4.else, Condition=@n3.isCtrlDown)
   > n8 = Cast To AVT_ItemData(Object=@n3.data)
   > Is Valid(InputObject=Get(self=@n8))
   > SyncAssetInCB!(
    execute
    self
    packageName:String="Get (a copy)"(Array=Get strings!(self:TA Generic List Item Data Object Reference=@n8) -> (strings:Array of Strings))
   ) -> (then)
   n9 = Cast To AVT_ItemData(execute=@n7.else, Object=@n3.data)
   n10 = Get(self=@n9)
   > n11 = Cast To StaticMesh(Object=@n10)
   > setCurrentMeshLodSettings!(execute, self, object:Object Reference=@n10) -> (then)
   n12 = Get borderLodSettings() -> (borderLodSettings:Border Object Reference)
   > SetVisibility(self=@n12, InVisibility=SelfHitTestInvisible)
   setCurrentMeshLodSettings(execute=@n11.CastFailed)
   > SetVisibility(self=@n12, InVisibility=Collapsed)
   "On Clicked (btnRemoveAllFromList)"!() -> (OutputDelegate, then, widget:User Widget Object Reference, event:Pointer Event Structure) {is=title K2Node_ComponentBoundEvent}
   > clearItems(self=Get TA_Generic01_List())
   > SaveGame_Save()
   > validateSettings()
   "On Clicked (btnFindBadAsset_InCB)"!() -> (OutputDelegate, then, widget:User Widget Object Reference, event:Pointer Event Structure) {is=title K2Node_ComponentBoundEvent}
   > clearItems(self=Get TA_Generic01_List())
   > n13 = findBadAssetsInCBSelection!(execute, self) -> (then, assetsAndReasons:Map of Object References to AVT Asset Data Struct Structures)
   > addAssetsToList(assetsAndReasons=@n13)
   > SaveGame_Save()
   > validateSettings()
   "On Clicked (btnShowFilters)"!() -> (OutputDelegate, then, widget:User Widget Object Reference, event:Pointer Event Structure) {is=title K2Node_ComponentBoundEvent}
   > Set setting_FiltersAreShown(setting_FiltersAreShown:bool=NOT Boolean(A=Get setting_FiltersAreShown() -> (setting_FiltersAreShown:bool))) -> (Output_Get:bool)
   > SaveGame_Save()
   > validateSettings()
   "On Clicked (btnExclusion_Edit)"!() -> (OutputDelegate, then, widget:User Widget Object Reference, event:Pointer Event Structure) {is=title K2Node_ComponentBoundEvent}
   > n14 = To Array!(execute, A:Set of Strings=Get setting_excludedPackageNames() -> (setting_excludedPackageNames:Set of Strings)) -> (then, Result:Array of Strings)
   > openPopup!(
    execute
    self:AVT Exclusion List Widget Object Reference=Get StaticMeshValidator_ExclusionList_Widget() -> (StaticMeshValidator_ExclusionList_Widget:AVT Exclusion List Widget Object Reference)
    strings:Array of Strings=@n14
   ) -> (then)
   n15 = "On Item Right Clicked (TA_Generic01_List)"!() -> (OutputDelegate, then, isShiftDown:bool, isCtrlDown:bool, data:Object Reference, item:User Widget Object Reference) {is=title K2Node_ComponentBoundEvent}
   > n16 = Cast To AVT_ItemData(Object=@n15.data)
   n17 = Get(self=@n16)
   > findMeshUseInWorld!(execute, self, meshPackageName:String=Split(SourceString=Get Object Path String(Object=@n17), InStr=., SearchDir=FromStart)) -> (then)
   "On Clicked (btnToggleClosed)"!() -> (OutputDelegate, then, widget:User Widget Object Reference, event:Pointer Event Structure) {is=title K2Node_ComponentBoundEvent}
   > n18 = Is Valid(InputObject=Get currentMeshForLodSettings() -> (currentMeshForLodSettings:Object Reference))
   setShow Mesh LODSettings!(execute=@n18.Is Not Valid, self, show:bool=false) -> (then)
   Set forceCloseLODSettings(execute=@n18.Is Valid, forceCloseLODSettings:bool=NOT Boolean(A=Get forceCloseLODSettings() -> (forceCloseLODSettings:bool))) -> (Output_Get:bool)
   n19 = Get SizeBox_3() -> (SizeBox_3:Size Box Object Reference)
   > setShow Mesh LODSettings(show=NOT Boolean(A=IsVisible(self=@n19)))
   "On Items Cleared (TA_Generic01_List)"() {is=title K2Node_ComponentBoundEvent}
   > setShow Mesh LODSettings(show=false)
   > SetVisibility(self=Get borderLodSettings(), InVisibility=Collapsed)
   "On Clicked (btnPrintToLog)"!() -> (OutputDelegate, then, widget:User Widget Object Reference, event:Pointer Event Structure) {is=title K2Node_ComponentBoundEvent}
   > printListEntries()
   "On Clicked (btnResetFilters)"!() -> (OutputDelegate, then, widget:User Widget Object Reference, event:Pointer Event Structure) {is=title K2Node_ComponentBoundEvent}
   > resetDisplayedFiltersToDefault()
   n20 = CustomEvent_0!() -> (OutputDelegate, then, strings:Array of Strings) {is=title K2Node_CustomEvent}
   n21 = Get setting_excludedPackageNames()
   > Clear!(execute, TargetSet:Set of Strings=@n21) -> (then) {is=title K2Node_CallFunction}
   > Add Items!(execute, TargetSet:Set of Strings=@n21, NewItems:Array of Strings=@n20.strings) -> (then)
   > SaveGame_Save()
   Event Construct()
   > Parent: Construct()
   > Bind Event to On Popup Submitted(self:AVT Exclusion List Widget Object Reference=Get StaticMeshValidator_ExclusionList_Widget(), Delegate=@n20) {is=title K2Node_AddDelegate}
   > setShow Mesh LODSettings(show=false)
  ]}
 SaveGame_PostLoad {nodes [
   n1 = SaveGame_PostLoad!() -> (then, saveGameExisted:bool, saveGame_Object:Save Game Object Reference) {is=title K2Node_FunctionEntry}
   > n2 = Sequence!(execute) -> (then_0, then_1, then_2:Exec)
   > Branch(Condition=@n1)
   > n3 = Cast To AVT_SaveGame!(execute, Object=@n1.saveGame_Object) -> (then, CastFailed, AsAVT Save Game:AVT Save Game Object Reference) {is=title K2Node_DynamicCast}
   > Set l_mySaveGame(l_mySaveGame:AVT Save Game Object Reference=@n3) -> (Output_Get:AVT Save Game Object Reference)
   n4 = Get filterData!(self:AVT Save Game Object Reference=Get l_mySaveGame() -> (l_mySaveGame:AVT Save Game Object Reference)) -> (filterData:Array of AVT Filter Data Struct Structures)
   > Set l_filterData(l_filterData:Array of AVT Filter Data Struct Structures=@n4) -> (Output_Get:Array of AVT Filter Data Struct Structures)
   > n5 = Sequence!(execute) -> (then_0, then_1, then_2, then_3:Exec)
   > Set setting_FiltersAreShown(setting_FiltersAreShown=Get filtersAreShown!(self:AVT Save Game Object Reference=Get l_mySaveGame()) -> (filtersAreShown:bool))
   > Set setting_excludedPackageNames(setting_excludedPackageNames:Set of Strings=Get excludedPackageNames!(self:AVT Save Game Object Reference=Get l_mySaveGame()) -> (excludedPackageNames:Set of Strings)) -> (Output_Get:Set of Strings)
   Branch(
    execute=@n5.then_1
    Condition="integer > integer"(A=Length(TargetArray:Array of AVT Filter Data Struct Structures=@n4), B=0) {is=title K2Node_CallFunction}
   )
   > n6 = For Each Loop with Break!(
    Exec:Exec
    Array:Array of Widget Object References=GetAllChildren(self=Get wrapBoxFilters() -> (wrapBoxFilters:Wrap Box Object Reference))
    Break:Exec=@n10.then
   ) -> (LoopBody:Exec, Array Element:Widget Object Reference, Array Index:int, Completed:Exec)
   > n7 = Cast To AVT_Base_FilterWidget!(execute, Object=@n6) -> (then, CastFailed, AsAVT Base Filter Widget:AVT Base Filter Widget Object Reference) {is=title K2Node_DynamicCast}
   > n8 = pickYourFilterData!(
    execute
    self:AVT Base Filter Widget Object Reference=@n7
    data:Array of AVT Filter Data Struct Structures=Get l_filterData() -> (l_filterData:Array of AVT Filter Data Struct Structures)
   ) -> (then, chosenIndex:int=0)
   > Branch(Condition="integer > integer"(A=@n8, B=-1))
   n9 = Get l_filterData()
   > Remove Index!(execute, TargetArray:Array of AVT Filter Data Struct Structures=@n9, IndexToRemove:int=@n8) -> (then)
   > n10 = Branch(Condition="integer < integer"!(A:int=Length(TargetArray=@n9), B:int=1) -> (ReturnValue:bool) {is=title K2Node_CallFunction})
   initFilterTypes(execute=@n2.then_1)
   > validateSettings()
   > n11 = For Each Loop(Array:Array of Widget Object References=GetAllChildren(self=Get wrapBoxFilters())) -> (Array Element:Widget Object Reference)
   > n12 = Cast To AVT_Base_FilterWidget(Object=@n11)
   > Bind Event to On Save Game Triggered(self:AVT Base Filter Widget Object Reference=@n12, Delegate=Create Event()) {is=title K2Node_AddDelegate}
   Branch(execute=@n2.then_2, Condition=@n1)
   Return Node!(execute, " ":bool=false)
  ]}
 SaveGame_PrepareToSave {nodes [
   SaveGame_PrepareToSave!() -> (then, SaveGameObject:Save Game Object Reference) {is=title K2Node_FunctionEntry}
   > n1 = Cast To AVT_SaveGame(Object=Get saveGame_Object() -> (saveGame_Object:Save Game Object Reference))
   > Set l_mySaveGame(l_mySaveGame=@n1)
   > n2 = Sequence()
   > Set filtersAreShown!(execute, filtersAreShown:bool=Get setting_FiltersAreShown(), self:AVT Save Game Object Reference=Get l_mySaveGame()) -> (then, Output_Get:bool)
   > Set excludedPackageNames!(execute, excludedPackageNames:Set of Strings=Get setting_excludedPackageNames(), self:AVT Save Game Object Reference=Get l_mySaveGame()) -> (then, Output_Get:Set of Strings)
   n3 = For Each Loop with Break(Exec=@n2.then_1, Array=GetAllChildren(self=Get wrapBoxFilters()))
   Set filterData!(
    execute=@n3.Completed
    filterData:Array of AVT Filter Data Struct Structures=Get l_filterData()
    self:AVT Save Game Object Reference=Get l_mySaveGame()
   ) -> (then, Output_Get:Array of AVT Filter Data Struct Structures)
   n4 = Cast To AVT_Base_FilterWidget(execute=@n3.LoopBody, Object=@n3)
   > n5 = getFilterData!(execute, self:AVT Base Filter Widget Object Reference=@n4) -> (then, data:AVT Filter Data Struct Structure)
   > Add(TargetArray:Array of AVT Filter Data Struct Structures=Get l_filterData(), NewItem:"AVT Filter Data Struct Structure (by ref)"=@n5) {is=title K2Node_CallArrayFunction}
   n6 = getListItems!(execute=@n2.then_2, self:TA Generic 01 List Object Reference=Get listFilters() -> (listFilters:TA Generic 01 List Object Reference)) -> (then, items:Array of User Widget Object References)
   > n7 = For Each Loop(Array:Array of User Widget Object References=@n6) -> (Array Element:User Widget Object Reference)
   Set activeFilterTypes!(
    execute=@n7.Completed
    activeFilterTypes:Array of AVT_FilterCategory_Enum Enums=Get l_activeFilterTypes() -> (l_activeFilterTypes:Array of AVT_FilterCategory_Enum Enums=NewEnumerator0)
    self:AVT Save Game Object Reference=Get l_mySaveGame()
   ) -> (then, Output_Get:Array of AVT_FilterCategory_Enum Enums=NewEnumerator0)
   n8 = Cast To TA_FilterTab_Button!(execute=@n7.LoopBody, Object=@n7) -> (then, CastFailed, AsTA Filter Tab Button:TA Filter Tab Button Object Reference) {is=title K2Node_DynamicCast}
   > Branch(Condition=getFilterEnabled!(self:TA Filter Tab Button Object Reference=@n8) -> (enabled:bool))
   > n9 = Keys!(
    execute
    TargetMap:Map of AVT_FilterCategory_Enum Enums to AVT Filter Type Children Struct Structures=Get filterCategoryMap() -> (filterCategoryMap:Map of AVT_FilterCategory_Enum Enums to AVT Filter Type Children Struct Structures=NewEnumerator0)
   ) -> (then, Keys:Array of AVT_FilterCategory_Enum Enums)
   > Add(
    TargetArray:Array of AVT_FilterCategory_Enum Enums=Get l_activeFilterTypes() -> (l_activeFilterTypes=NewEnumerator0)
    NewItem:"AVT_FilterCategory_Enum Enum (by ref)"="Get (a copy)"(Array:Array of AVT_FilterCategory_Enum Enums=@n9, Dimension 1=@n7.Array Index) -> (Output:AVT_FilterCategory_Enum Enum)
   )
   n10 = getSelectedIndex!(execute=@n2.then_3, self:TA Generic 01 List Object Reference=Get listFilters()) -> (then, index:int=0)
   > n11 = Keys(TargetMap=Get filterCategoryMap() -> (filterCategoryMap=NewEnumerator0))
   > Set currentSelectedFilter!(
    execute
    currentSelectedFilter:AVT_FilterCategory_Enum Enum="Get (a copy)"(Array=@n11, Dimension 1=@n10)
    self:AVT Save Game Object Reference=Get l_mySaveGame()
   ) -> (then, Output_Get:AVT_FilterCategory_Enum Enum=NewEnumerator0)
   Return Node(" "=false)
  ]}
 SyncAssetInCB {nodes [
   n1 = SyncAssetInCB!() -> (then, packageName:String) {is=title K2Node_FunctionEntry}
   > SyncBrowserToObjects(AssetPaths=Make Array("[0]":String=@n1) -> (Array:Array of Strings))
  ]}
 addAssetsToList {nodes [
   n1 = addAssetsToList!() -> (then, assetsAndReasons:Map of Object References to AVT Asset Data Struct Structures) {is=title K2Node_FunctionEntry}
   > n2 = Keys(TargetMap:Map of Object References to AVT Asset Data Struct Structures=@n1) -> (Keys:Array of Object References)
   > n3 = For Each Loop(Array:Array of Object References=@n2) -> (Array Element:Object Reference)
   > n4 = Values!(execute, TargetMap:Map of Object References to AVT Asset Data Struct Structures=@n1) -> (then, Values:Array of AVT Asset Data Struct Structures)
   > n5 = Construct Object from Class(Class=/Game/Blueprints/Blutilities/ForArtists/AssetValidationTool/Data/AVT_ItemData.AVT_ItemData_C, self=Self-Reference()) -> (ReturnValue:AVT Item Data Object Reference)
   > Set displayString!(execute, displayString:String=GetDisplayName(Object=@n3), self:TA Generic List Item Data Object Reference=@n5) -> (then, Output_Get:String)
   > Set assetErrorData!(
    execute
    assetErrorData:AVT Asset Data Struct Structure="Get (a ref)"(Array:Array of AVT Asset Data Struct Structures=@n4, Dimension 1=@n3.Array Index) -> (Output:"AVT Asset Data Struct Structure (by ref)")
    self:AVT Item Data Object Reference=@n5
   ) -> (then, Output_Get:AVT Asset Data Struct Structure)
   n6 = CreateAssetData!(InAsset=@n3, bAllowBlueprintClass=false) -> (
    ReturnValue_PackageName:Name
    ReturnValue_PackagePath:Name
    ReturnValue_AssetName:Name
    ReturnValue_AssetClassPath:Top Level Asset Path Structure
   )
   > Set strings!(execute, strings:Array of Strings=Make Array("[0]"="To String (Name)"(InName=@n6)), self:TA Generic List Item Data Object Reference=@n5) -> (then, Output_Get:Array of Strings)
   > addItem!(execute, self:TA Generic 01 List Object Reference=Get TA_Generic01_List(), dataObject:Object Reference=@n5) -> (then, newItem:User Widget Object Reference, newData:Object Reference)
  ]}
 addFilterTypeButton {nodes [
   n1 = Cast To AVT_SaveGame(Object=Get saveGame_Object())
   n2 = Construct Object from Class(
    execute=@n6.then_1
    Class=/Game/Blueprints/Blutilities/ForArtists/AssetValidationTool/Data/AVT_FilterCategory_ItemData.AVT_FilterCategory_ItemData_C
    self=Self-Reference()
   ) -> (ReturnValue:AVT Filter Category Item Data Object Reference)
   n3 = addItem(execute=@n12.then, self=Get listFilters(), dataObject=@n2)
   > n4 = Cast To TA_FilterTab_Button(Object=@n3)
   setFilterEnabled!(
    execute=@n1.then
    self:TA Filter Tab Button Object Reference=@n4
    enabled:bool=Contains Item(
     TargetArray:Array of AVT_FilterCategory_Enum Enums=Get activeFilterTypes!(self:AVT Save Game Object Reference=@n1) -> (activeFilterTypes:Array of AVT_FilterCategory_Enum Enums=NewEnumerator0)
     ItemToFind:"AVT_FilterCategory_Enum Enum (by ref)"=Get filterCategory() -> (filterCategory:AVT_FilterCategory_Enum Enum=NewEnumerator0)
    )
   ) -> (then)
   setFilterEnabled(execute=@n1.CastFailed, self=@n4, enabled=true)
   n5 = addFilterTypeButton!() -> (then, displayString:String, filterCategory22:Name, filterCategory:AVT_FilterCategory_Enum Enum=NewEnumerator0) {is=title K2Node_FunctionEntry}
   > n6 = Sequence!(execute) -> (then_0, then_1)
   > n7 = SaveGame_DoesSaveExist!(execute, self:TA Base Editor Widget Object Reference) -> (then, exists:bool)
   > n8 = Branch(Condition=@n7)
   > n9 = Cast To AVT_SaveGame(Object=Get saveGame_Object())
   > Set l_enable(l_enable:bool=Contains Item(
     TargetArray=Get activeFilterTypes(self=@n9) -> (activeFilterTypes=NewEnumerator0)
     ItemToFind=Get filterCategory() -> (filterCategory=NewEnumerator0)
    )) -> (Output_Get:bool)
   Set l_enable(execute=@n8.else, l_enable=true)
   Bind Event to On Checked Status Changed(execute=@n4.then, self:TA Filter Tab Button Object Reference=@n4, Delegate=Create Event()) {is=title K2Node_AddDelegate}
   n10 = Make Array("[0]":bool=Get l_enable() -> (l_enable:bool)) -> (Array:Array of Booleans)
   n11 = Get filterCategory() -> (filterCategory=NewEnumerator0)
   Set displayString(execute=@n2.then, displayString=@n5, self=@n2)
   > Set filterCategory!(execute, filterCategory:AVT_FilterCategory_Enum Enum=@n11, self:AVT Filter Category Item Data Object Reference=@n2) -> (then, Output_Get:AVT_FilterCategory_Enum Enum=NewEnumerator0)
   > n12 = Set bools!(execute, bools:Array of Booleans=@n10, self:TA Generic List Item Data Object Reference=@n2) -> (then, Output_Get:Array of Booleans)
  ]}
 checkForExclusion {nodes [
   n1 = Inputs!() -> (execute, data:Asset Data Structure)
   n2 = "To String (Name)"(InName=Break Asset Data(AssetData=@n1))
   n3 = Get setting_excludedPackageNames()
   > n4 = Branch(Condition=OR Boolean(
     A=Contains Item!(TargetSet:Set of Strings=@n3, ItemToFind:"String (by ref)"=@n2) -> (ReturnValue:bool) {is=title K2Node_CallFunction}
     B=Contains(SearchIn=@n2, Substring=HLOD)
    ))
   Outputs!(Allow:Exec=@n4.else, Excluded:Exec=@n4.then)
  ]}
 excludeSelectedListItem {nodes [
   excludeSelectedListItem() {is=title K2Node_FunctionEntry}
   > n1 = Sequence!(execute) -> (then_0, then_1, then_2:Exec)
   > n2 = getSelectedItems!(execute, self:TA Generic 01 List Object Reference=Get TA_Generic01_List()) -> (then, isAnySelected:bool, items:Array of User Widget Object References, selectedData:Array of Object References)
   > n3 = For Each Loop(Array=@n2.selectedData)
   > n4 = Cast To AVT_ItemData(Object=@n3)
   n5 = Get(self=@n4)
   > Add!(
    execute
    TargetSet:Set of Strings=Get setting_excludedPackageNames()
    NewItem:"String (by ref)"="To String (Name)"(InName=CreateAssetData(InAsset=@n5, bAllowBlueprintClass=false))
   ) -> (then) {is=title K2Node_CallFunction}
   n6 = getSelectedIndex(execute=@n1.then_1, self=Get TA_Generic01_List()) -> (index=0)
   > Set l_index(l_index:int=@n6) -> (Output_Get:int=0)
   > n7 = getSelectedItems(self=Get TA_Generic01_List())
   > Branch(Condition=@n7)
   > removeItems!(execute, self:TA Generic 01 List Object Reference=Get TA_Generic01_List(), itemWidgets:Array of User Widget Object References=@n7.items) -> (then)
   > setIndexSelected!(
    execute
    self:TA Generic 01 List Object Reference=Get TA_Generic01_List()
    index:int=Get l_index() -> (l_index:int=0)
    selected:bool=true
    throwEvent:bool=true
   ) -> (then)
   SaveGame_Save(execute=@n1.then_2)
  ]}
 findBadAssetsInCBSelection {nodes [
   findBadAssetsInCBSelection() {is=title K2Node_FunctionEntry}
   > n1 = GetSelectedAssetData()
   > n2 = Set l_Assets(l_Assets:Array of Asset Data Structures=@n1) -> (Output_Get:Array of Asset Data Structures)
   > RunAssetsThroughFilter!(
    execute
    self:Asset Registry Interface=GetAssetRegistry()
    AssetDataList:Array of Asset Data Structures=@n2
    Filter:"ARFilter Structure (by ref)"=MakeARFilter!(
     PackageNames:Array of Names
     PackagePaths:Array of Names
     SoftObjectPaths:Array of Soft Object Path Structures
     ClassPaths:Array of Top Level Asset Path Structures
     RecursiveClassPathsExclusionSet:Set of Top Level Asset Path Structures
     ClassNames:Array of Names=Make Array!("[0]":Name=StaticMesh, "[1]":Name=SkeletalMesh, "[2]":Name=Texture2D) -> (Array:Array of Names)
     RecursiveClassesExclusionSet:Set of Names
     bRecursivePaths:bool=false
     bRecursiveClasses:bool=true
     bIncludeOnlyOnDiskAssets:bool=false
    ) -> (ReturnValue:ARFilter Structure)
   ) -> (then)
   > n3 = For Each Loop(Array:Array of Asset Data Structures=Get l_Assets() -> (l_Assets:Array of Asset Data Structures)) -> (Array Element:Asset Data Structure)
   Return Node!(
    execute=@n3.Completed
    assetsAndReasons:Map of Object References to AVT Asset Data Struct Structures=Get l_assetMap() -> (l_assetMap:Map of Object References to AVT Asset Data Struct Structures)
   )
   > n4 = Validate Asset Is Bad!(" ":Exec, asset:Object Reference=GetAsset(InAssetData=@n3)) -> ("bad(AddIt)":Exec, "good(DontAdd)":Exec, outMesh:Static Mesh Object Reference, errors:Array of AVT Error Struct Structures) {is=title K2Node_MacroInstance}
   > Add!(
    execute
    TargetMap:Map of Object References to AVT Asset Data Struct Structures=Get l_assetMap()
    Key:"Object Reference (by ref)"=@n4
    Value:"AVT Asset Data Struct Structure (by ref)"=Make AVT Asset Data Struct!(
     asset_39_#1:Object Reference=@n4
     errors_35_#2:Array of AVT Error Struct Structures=@n4.errors
     csvLine_42_#3:String=Get Object Path String(Object=@n4)
     assetPackageName_46_#4:String
    ) -> (AVT_AssetData_Struct:AVT Asset Data Struct Structure) {is=title K2Node_MakeStruct}
   ) -> (then)
  ]}
 findBadMeshesInLevel {nodes [
   n1 = GetAllLevelActorsComponents!(execute) -> (then, ReturnValue:Array of Actor Component Object References)
   > n2 = Filter by Class!(
    execute
    TargetArray:Array of Object References=@n1
    ObjectClass:Object Class Reference=&StaticMeshComponent
    FilterType:EEditorScriptingFilterType Enum
   ) -> (then, ReturnValue:Array of Static Mesh Component Object References)
   > n3 = Set l_meshComps(l_meshComps:Array of Static Mesh Component Object References=@n2) -> (Output_Get:Array of Static Mesh Component Object References)
   n4 = For Each Loop(
    Exec=[@n16.Completed, @n13.Completed, @n3.then]
    Array:Array of Static Mesh Component Object References=Get l_meshComps() -> (l_meshComps:Array of Static Mesh Component Object References)
   ) -> (Array Element:Static Mesh Component Object Reference)
   Return Node!(
    execute=@n4.Completed
    meshesAndReasons:Map of Static Mesh Object References to AVT Asset Data Struct Structures=Get l_meshes() -> (l_meshes:Map of Static Mesh Object References to AVT Asset Data Struct Structures)
   )
   n5 = Get StaticMesh(self=@n4)
   > Is Valid(InputObject=@n5)
   n6 = Get bHiddenInGame!(self:Scene Component Object Reference=@n4) -> (bHiddenInGame:bool)
   > Branch(Condition=AND Boolean(A=Get bVisible!(self:Scene Component Object Reference=@n4) -> (bVisible:bool), B=NOT Boolean(A=@n6)))
   > Check for Exclusion!(execute:Exec, data:Asset Data Structure=CreateAssetData!(InAsset=@n5, bAllowBlueprintClass=false) -> (ReturnValue:Asset Data Structure)) -> (Allow:Exec, Excluded:Exec) {is=title K2Node_MacroInstance}
   > n7 = Validate Asset Is Bad(asset=@n5)
   > Add(
    TargetMap:Map of Static Mesh Object References to AVT Asset Data Struct Structures=Get l_meshes()
    Key:"Static Mesh Object Reference (by ref)"=@n7
    Value=Make AVT Asset Data Struct(
     asset_39_#1=@n7
     errors_35_#2=@n7.errors
     csvLine_42_#3=Get Object Path String(Object=@n7)
    )
   )
   findBadMeshesInLevel() {is=title K2Node_FunctionEntry}
   > n8 = GetPIEWorlds!(execute, bIncludeDedicatedServer:bool=false) -> (then, ReturnValue:Array of World Object References)
   > n9 = Branch(Condition=Is Valid Index!(TargetArray:Array of World Object References=@n8, IndexToTest:int) -> (ReturnValue:bool))
   n10 = "Get (a copy)"(Array:Array of World Object References=@n8) -> (Output:World Object Reference)
   > PrintString(InString=Append(A="Worlds ", B=GetDisplayName(Object=@n10)))
   > n11 = GetAllActorsOfClass!(execute, WorldContextObject:Object Reference=@n10, ActorClass:Actor Class Reference=&Actor) -> (then, OutActors:Array of Actor Object References)
   n12 = Length(TargetArray:Array of Actor Object References=@n11)
   > PrintString(InString=Append(A="Actors ", B="To String (Integer)"(InInt=@n12)))
   > n13 = For Each Loop(Array:Array of Actor Object References=@n11) -> (Array Element:Actor Object Reference)
   > Append Array!(
    execute
    TargetArray:Array of Static Mesh Component Object References=Get l_meshComps()
    SourceArray:Array of Static Mesh Component Object References=Get Components By Class!(self:Actor Object Reference=@n13, ComponentClass:Actor Component Class Reference=&StaticMeshComponent) -> (ReturnValue:Array of Static Mesh Component Object References)
   ) -> (then)
   n14 = GetEditorWorld(execute=@n9.else, self=Get UnrealEditorSubsystem())
   > n15 = GetAllActorsOfClass(WorldContextObject=@n14, ActorClass=&Actor)
   > n16 = For Each Loop(Array=@n15)
   > Append Array(TargetArray=Get l_meshComps(), SourceArray=Get Components By Class(self=@n16, ComponentClass=&StaticMeshComponent))
   > Branch(Condition=Contains(SearchIn=GetDisplayName(Object=@n16), Substring=CATTUS))
   > PrintString(InString="Found one!")
  ]}
 findMeshUseInWorld {nodes [
   n1 = findMeshUseInWorld!() -> (then, meshPackageName:String) {is=title K2Node_FunctionEntry}
   n2 = String To Name(InString=@n1)
   > n3 = GetAssets!(
    execute
    self:Asset Registry Interface=GetAssetRegistry()
    Filter:"ARFilter Structure (by ref)"=MakeARFilter(
     PackageNames=Make Array!("[0]"=@n2) -> (Array)
     ClassNames=Make Array!("[0]"=StaticMesh, "[1]":Name=SkeletalMesh) -> (Array)
     bRecursivePaths=false
     bRecursiveClasses=false
     bIncludeOnlyOnDiskAssets=false
    )
    bSkipARFilteredAssets:bool=true
   ) -> (then, OutAssetData:Array of Asset Data Structures, ReturnValue:bool)
   > n4 = For Each Loop(Array:Array of Asset Data Structures=@n3) -> (Array Element:Asset Data Structure)
   SetSelectedLevelActors(
    execute=@n4.Completed
    self=Get EditorActorSubsystem()
    ActorsToSelect=Get l_finalActors() -> (l_finalActors:Array of Actor Object References)
   )
   n5 = Sequence!(execute=@n4.LoopBody) -> (then_0, then_1)
   > n6 = GetAllLevelActorsComponents!(execute, self:Editor Actor Subsystem Object Reference=Get EditorActorSubsystem()) -> (then, ReturnValue)
   > n7 = Filter by Class(TargetArray=@n6, ObjectClass=&StaticMeshComponent)
   > Append Array(
    TargetArray:Array of Actor Component Object References=Get l_components() -> (l_components:Array of Actor Component Object References)
    SourceArray:Array of Actor Component Object References=@n7
   )
   n8 = For Each Loop(Exec=@n5.then_1, Array:Array of Actor Component Object References=Get l_components()) -> (Array Element:Actor Component Object Reference)
   > n9 = Cast To StaticMeshComponent(Object=@n8)
   > n10 = Set l_object(l_object:Object Reference=Get StaticMesh(self=@n9)) -> (Output_Get:Object Reference)
   Branch(
    execute=[@n12.then, @n10.then]
    Condition="Equal (Object)"(A=Get l_object() -> (l_object:Object Reference), B=GetAsset(InAssetData=@n4)) {is=title K2Node_CallFunction}
   )
   > Add(TargetArray:Array of Actor Object References=Get l_finalActors(), NewItem:"Actor Object Reference (by ref)"=GetOwner(self=@n8)) {is=title K2Node_CallArrayFunction}
   n11 = Cast To SkeletalMeshComponent!(execute=@n9.CastFailed, Object=@n8) -> (then, CastFailed, AsSkeletal Mesh Component:Skeletal Mesh Component Object Reference) {is=title K2Node_DynamicCast}
   > n12 = Set l_object(l_object=Get SkeletalMesh!(self:Skinned Mesh Component Object Reference=@n11) -> (SkeletalMesh:Skeletal Mesh Object Reference))
  ]}
 getPrintStringPerEntry {nodes [
   n1 = getPrintStringPerEntry!() -> (then, entryData:Object Reference, titles:Array of Strings) {is=title K2Node_FunctionEntry}
   > n2 = Sequence!(execute) -> (then_0, then_1, then_2:Exec)
   > n3 = For Each Loop(Array:Array of Strings=@n1.titles) -> (Array Element:String)
   > Add(
    TargetMap:Map of Strings to Strings=Get l_titlesMap() -> (l_titlesMap:Map of Strings to Strings)
    Key:"String (by ref)"=@n3
    Value:"String (by ref)"
   ) {is=title K2Node_CallFunction}
   n4 = Cast To AVT_ItemData(execute=@n2.then_1, Object=@n1)
   n5 = Get(self=@n4)
   > n6 = For Each Loop(Array:Array of AVT Error Struct Structures=@n5.assetErrorData_errors_35_#2) -> (Array Element:AVT Error Struct Structure)
   n7 = Break AVT Error Struct!(AVT_Error_Struct:"AVT Error Struct Structure (by ref)"=@n6) -> (
    title_32_#5:String
    description_35_#6:String
    tooltip_37_#7:String
    color_40_#8:Linear Color Structure
    csvLine_46_#9:String
    csvHeader_49_#10:String
    optionalAssetData_52_#11:Asset Data Structure
    optionalTypeTag_55_#12:String
    optionalOriginPackageName_62_#13:Name
    helpCommand_65_#14:String
   ) {is=title K2Node_BreakStruct}
   > Add(TargetMap=Get l_titlesMap(), Key=@n7.csvHeader_49_#10, Value=@n7.csvLine_46_#9)
   Set l_csv(execute=@n2.then_2, l_csv:String=@n5.assetErrorData_csvLine_42_#3) -> (Output_Get:String)
   > n8 = Values(TargetMap:Map of Strings to Strings=Get l_titlesMap()) -> (Values:Array of Strings)
   > n9 = For Each Loop(Array:Array of Strings=@n8) -> (Array Element:String)
   Return Node!(execute=@n9.Completed, entryCSV:String=Get l_csv() -> (l_csv:String))
   > Set l_csv(l_csv=Append!(A=Get l_csv(), B=",", C:String=@n9) -> (ReturnValue))
  ]}
 getPrintString_Titles {nodes [
   getPrintString_Titles() {is=title K2Node_FunctionEntry}
   > n1 = getListData(self=Get TA_Generic01_List())
   > n2 = For Each Loop(Array:Array of Object References=@n1) -> (Array Element:Object Reference)
   Set l_titleCsv(execute=@n2.Completed, l_titleCsv:String=Asset) -> (Output_Get:String)
   > n3 = To Array(A=Get l_titles() -> (l_titles:Set of Strings))
   > n4 = For Each Loop(Array:Array of Strings=@n3) -> (Array Element:String)
   Return Node!(execute=@n4.Completed, titles:Array of Strings=@n3, csv:String=Get l_titleCsv() -> (l_titleCsv:String))
   > Set l_titleCsv(l_titleCsv=Append(A=Get l_titleCsv(), B=",", C=@n4))
   n5 = Cast To AVT_ItemData(execute=@n2.LoopBody, Object=@n2)
   > n6 = For Each Loop(Array:Array of AVT Error Struct Structures=Get(self=@n5).assetErrorData_errors_35_#2) -> (Array Element:AVT Error Struct Structure)
   > Add!(
    execute
    TargetSet:Set of Strings=Get l_titles()
    NewItem:"String (by ref)"=Break AVT Error Struct(AVT_Error_Struct=@n6).csvHeader_49_#10
   ) -> (then)
  ]}
 handleFilterTypeChanged {nodes [
   handleFilterTypeChanged() {is=title K2Node_FunctionEntry}
   > n1 = getSelectedData!(execute, self:TA Generic 01 List Object Reference=Get listFilters()) -> (then, selectedData:Array of Object References)
   > n2 = Cast To AVT_FilterCategory_ItemData!(execute, Object="Get (a copy)"(Array:Array of Object References=@n1) -> (Output:Object Reference)) -> (then, CastFailed, AsAVT Filter Category Item Data:AVT Filter Category Item Data Object Reference) {is=title K2Node_DynamicCast}
   > n3 = For Each Loop(Array:Array of Widget Object References=GetAllChildren(self=Get wrapBoxFilters())) -> (Array Element:Widget Object Reference)
   > n4 = Cast To AVT_Base_FilterWidget(Object=@n3)
   n5 = Get filterCategory!(self:AVT Filter Category Item Data Object Reference=@n2) -> (filterCategory=NewEnumerator0)
   n6 = Get filterType!(self:AVT Base Filter Widget Object Reference=@n4) -> (filterType:AVT_FilterCategory_Enum Enum=NewEnumerator0)
   > SetVisibility(
    self=@n4
    InVisibility=Select(Option 0=Collapsed, Option 1=Visible, Index="Equal (Enum)"(A:AVT_FilterCategory_Enum Enum=@n6, B:AVT_FilterCategory_Enum Enum=@n5))
   )
  ]}
 handleFilterTypeCheckStatus {nodes [
   n1 = handleFilterTypeCheckStatus!() -> (then, widget:TA Filter Tab Button Object Reference, data:Object Reference, isChecked:bool=~, isShiftDown:bool=~, isControlDown:bool=~) {is=title K2Node_FunctionEntry}
   > n2 = Sequence!(execute) -> (then_0, then_1)
   > n3 = Branch(Condition=@n1.isShiftDown)
   > n4 = getListItems(self=Get listFilters())
   > n5 = For Each Loop(Array:Array of User Widget Object References=@n4) -> (Array Element:User Widget Object Reference)
   > n6 = Cast To TA_FilterTab_Button(Object=@n5)
   > setFilterEnabled(self=@n6, enabled=@n1.isChecked)
   Branch(execute=@n3.else, Condition=@n1.isControlDown)
   > n7 = Cast To AVT_FilterCategory_ItemData(Object=@n1.data)
   > setAllFiltersEnabledByCategory!(
    execute
    self
    enabled:bool=true
    filterCategory:AVT_FilterCategory_Enum Enum=Get filterCategory(self=@n7) -> (filterCategory=NewEnumerator0)
    isToggle:bool=true
   ) -> (then)
   SaveGame_Save(execute=@n2.then_1)
  ]}
 initFilterTypes {nodes [
   initFilterTypes() {is=title K2Node_FunctionEntry}
   > Clear!(
    execute
    TargetMap:Map of AVT_FilterCategory_Enum Enums to AVT Filter Type Children Struct Structures=Get filterCategoryMap() -> (filterCategoryMap=NewEnumerator0)
   ) -> (then)
   > clearItems(self=Get listFilters())
   > n1 = Sequence!(execute) -> (then_0, then_1, then_2:Exec)
   > Set l_children(l_children:Array of Widget Object References=GetAllChildren(self=Get wrapBoxFilters())) -> (Output_Get:Array of Widget Object References)
   > n2 = ForEach AVT_FilterCategory_Enum!(execute:Exec, SkipHidden:bool) -> (LoopBody:Exec, EnumValue:AVT_FilterCategory_Enum Enum, then:Exec) {is=title K2Node_ForEachElementInEnum}
   > Set l_category(l_category:AVT_FilterCategory_Enum Enum=@n2) -> (Output_Get:AVT_FilterCategory_Enum Enum=NewEnumerator0)
   > n3 = Reverse for Each Loop!(Exec:Exec, Array:Array of Widget Object References=Get l_children() -> (l_children:Array of Widget Object References)) -> (LoopBody:Exec, ArrayIndex:int, ArrayElement:Widget Object Reference, Completed:Exec) {is=title K2Node_MacroInstance}
   > n4 = Cast To AVT_Base_FilterWidget(Object=@n3.ArrayElement)
   > Branch(Condition="Equal (Enum)"(A=Get filterType(self=@n4) -> (filterType=NewEnumerator0), B=Get l_category() -> (l_category:AVT_FilterCategory_Enum Enum=NewEnumerator0)))
   > Remove Index(TargetArray:Array of Widget Object References=Get l_children(), IndexToRemove=@n3)
   n5 = Find(
    TargetMap:Map of AVT_FilterCategory_Enum Enums to AVT Filter Type Children Struct Structures=Get filterCategoryMap() -> (filterCategoryMap=NewEnumerator0)
    Key:"AVT_FilterCategory_Enum Enum (by ref)"=Get l_category() -> (l_category=NewEnumerator0)
   ) -> (Value:AVT Filter Type Children Struct Structure)
   > n6 = Branch(Condition=@n5.ReturnValue)
   n7 = Break AVT Filter Type Children Struct!(AVT_FilterTypeChildren_Struct:"AVT Filter Type Children Struct Structure (by ref)"=@n5) -> (
    filterCategory_22_#15:AVT_FilterCategory_Enum Enum
    children_20_#16:Set of AVT Base Filter Widget Object References
   ) {is=title K2Node_BreakStruct}
   > n8 = Union!(
    execute
    A:Set of AVT Base Filter Widget Object References=@n7.children_20_#16
    B:Set of AVT Base Filter Widget Object References=Make Set!("[0]":AVT Base Filter Widget Object Reference=@n4) -> (Set:Set of AVT Base Filter Widget Object References) {is=title K2Node_MakeSet}
   ) -> (then, Result:Set of AVT Base Filter Widget Object References)
   > Add!(
    execute
    TargetMap:Map of AVT_FilterCategory_Enum Enums to AVT Filter Type Children Struct Structures=Get filterCategoryMap() -> (filterCategoryMap=NewEnumerator0)
    Key:"AVT_FilterCategory_Enum Enum (by ref)"=Get l_category() -> (l_category=NewEnumerator0)
    Value:"AVT Filter Type Children Struct Structure (by ref)"=Make AVT Filter Type Children Struct!(
     filterCategory_22_#15:AVT_FilterCategory_Enum Enum=@n7
     children_20_#16:Set of AVT Base Filter Widget Object References=@n8
    ) -> (AVT_FilterTypeChildren_Struct:AVT Filter Type Children Struct Structure) {is=title K2Node_MakeStruct}
   ) -> (then)
   Add(
    execute=@n6.else
    TargetMap=Get filterCategoryMap() -> (filterCategoryMap=NewEnumerator0)
    Key=Get l_category() -> (l_category=NewEnumerator0)
    Value=Make AVT Filter Type Children Struct(
     filterCategory_22_#15=Get l_category() -> (l_category=NewEnumerator0)
     children_20_#16=Make Set("[0]"=@n4)
    )
   )
   n9 = Values(
    execute=@n1.then_1
    TargetMap:Map of AVT_FilterCategory_Enum Enums to AVT Filter Type Children Struct Structures=Get filterCategoryMap() -> (filterCategoryMap=NewEnumerator0)
   ) -> (Values:Array of AVT Filter Type Children Struct Structures)
   > n10 = For Each Loop!(Exec, Array:Array of AVT Filter Type Children Struct Structures=@n9) -> (
    LoopBody
    Array Element_filterCategory_22_#15:AVT_FilterCategory_Enum Enum
    Array Element_children_20_#16:Set of AVT Base Filter Widget Object References
    Array Index
    Completed
   )
   > Branch(Condition="integer > integer"(
     A=Length!(TargetSet:Set of AVT Base Filter Widget Object References=@n10.Array Element_children_20_#16) -> (ReturnValue:int=0) {is=title K2Node_CallFunction}
     B=0
    ))
   n11 = Enum to String!(Enumerator:AVT_FilterCategory_Enum Enum=@n10) -> (ReturnValue:String)
   > addFilterTypeButton!(
    execute
    self
    displayString:String=@n11
    filterCategory22:Name=String To Name(InString=@n11)
    filterCategory:AVT_FilterCategory_Enum Enum=@n10
   ) -> (then) {is=title K2Node_CallFunction}
   n12 = Cast To AVT_SaveGame(execute=@n1.then_2, Object=Get saveGame_Object())
   > n13 = Keys(TargetMap:Map of AVT_FilterCategory_Enum Enums to AVT Filter Type Children Struct Structures=Get filterCategoryMap() -> (filterCategoryMap=NewEnumerator0)) -> (Keys:Array of AVT_FilterCategory_Enum Enums)
   > setIndexSelected(
    self=Get listFilters()
    index=Find Item!(
     TargetArray:Array of AVT_FilterCategory_Enum Enums=@n13
     ItemToFind:"AVT_FilterCategory_Enum Enum (by ref)"=Get currentSelectedFilter!(self:AVT Save Game Object Reference=@n12) -> (currentSelectedFilter:AVT_FilterCategory_Enum Enum=NewEnumerator0)
    ) -> (ReturnValue)
    selected=true
    throwEvent=true
   )
   > handleFilterTypeChanged() {is=title K2Node_CallFunction}
   setIndexSelected(execute=@n12.CastFailed, self=Get listFilters(), index=0, selected=true, throwEvent=true)
   > handleFilterTypeChanged()
  ]}
 openAssetForEdit {nodes [
   n1 = openAssetForEdit!() -> (then, Asset:Object Reference) {is=title K2Node_FunctionEntry}
   > OpenEditorForAssets!(
    execute
    self:Asset Editor Subsystem Object Reference=Get AssetEditorSubsystem()
    Assets:Array of Object References=Make Array!("[0]":Object Reference=@n1) -> (Array:Array of Object References)
    OpenedMethod:EAssetTypeActivationOpenedMethod Enum=Edit
   ) -> (then, ReturnValue:bool)
  ]}
 printListEntries {nodes [
   printListEntries() {is=title K2Node_FunctionEntry}
   > n1 = Sequence!(execute) -> (then_0, then_1, then_2, then_3:Exec)
   > n2 = getPrintString_Titles!(execute, self) -> (then, titles:Array of Strings, csv:String) {is=title K2Node_CallFunction}
   > Set l_titles(l_titles:Array of Strings=@n2) -> (Output_Get:Array of Strings)
   > Add(TargetArray:Array of Strings=Get l_entries() -> (l_entries:Array of Strings), NewItem:"String (by ref)"=@n2.csv) {is=title K2Node_CallArrayFunction}
   n3 = getListData(execute=@n1.then_1, self=Get TA_Generic01_List())
   > n4 = For Each Loop!(Exec, Array:Array of Object References=@n3) -> (LoopBody, Array Element:Object Reference, Array Index, Completed)
   > n5 = getPrintStringPerEntry!(execute, self, entryData:Object Reference=@n4, titles:Array of Strings=Get l_titles() -> (l_titles:Array of Strings)) -> (then, entryCSV:String) {is=title K2Node_CallFunction}
   > Add(TargetArray=Get l_entries(), NewItem=@n5)
   Set l_final(execute=@n1.then_2, l_final:String="\r\n") -> (Output_Get:String)
   > n6 = For Each Loop(Array:Array of Strings=Get l_entries()) -> (Array Element:String)
   > Set l_final(l_final=Append(A=Get l_final() -> (l_final:String), B="\r\n", C=@n6))
   PrintString(execute=@n1.then_3, InString=Get l_final())
  ]}
 resetDisplayedFiltersToDefault {nodes [
   resetDisplayedFiltersToDefault() {is=title K2Node_FunctionEntry}
   > n1 = Sequence()
   > n2 = getSelectedItems(self=Get listFilters())
   > Branch(Condition=@n2)
   > n3 = Cast To AVT_FilterCategory_ItemData(Object="Get (a copy)"(Array=@n2.selectedData))
   > Set l_filterCategory(l_filterCategory:AVT_FilterCategory_Enum Enum=Get filterCategory(self=@n3) -> (filterCategory=NewEnumerator0)) -> (Output_Get:AVT_FilterCategory_Enum Enum=NewEnumerator0)
   n4 = Get filterCategoryMap() -> (filterCategoryMap=NewEnumerator0)
   n5 = Get l_filterCategory() -> (l_filterCategory:AVT_FilterCategory_Enum Enum=NewEnumerator0)
   n6 = To Array(
    execute=@n1.then_1
    A:Set of AVT Base Filter Widget Object References=Break AVT Filter Type Children Struct(AVT_FilterTypeChildren_Struct=Find(TargetMap=@n4, Key=@n5)).children_20_#16
   ) -> (Result:Array of AVT Base Filter Widget Object References)
   > Set l_filterWidgets(l_filterWidgets:Array of AVT Base Filter Widget Object References=@n6) -> (Output_Get:Array of AVT Base Filter Widget Object References)
   Branch(
    execute=@n1.then_2
    Condition=Is Valid Index(TargetArray:Array of AVT Base Filter Widget Object References=Get l_filterWidgets() -> (l_filterWidgets:Array of AVT Base Filter Widget Object References))
   )
   n7 = "Get (a copy)"(Array:Array of AVT Base Filter Widget Object References=Get l_filterWidgets()) -> (Output:AVT Base Filter Widget Object Reference)
   > Set l_enable(l_enable=NOT Boolean(A=getFilterEnabled!(self:AVT Base Filter Widget Object Reference=@n7) -> (isEnabled:bool)))
   n8 = For Each Loop(Exec=@n1.then_3, Array:Array of AVT Base Filter Widget Object References=Get l_filterWidgets()) -> (Array Element:AVT Base Filter Widget Object Reference)
   > setFilterEnabled!(execute, self:AVT Base Filter Widget Object Reference=@n8, isEnabled:bool=Get l_enable()) -> (then, changed:bool)
  ]}
 runAssetTests {nodes [
   n1 = Union(
    A:Set of Widget Object References=Get l_filtersToRun() -> (l_filtersToRun:Set of Widget Object References)
    B:Set of Widget Object References
   ) -> (Result:Set of Widget Object References)
   > Set l_filtersToRun(l_filtersToRun:Set of Widget Object References=@n1) -> (Output_Get:Set of Widget Object References)
   n2 = runAssetTests!() -> (then, assetMeshToCheck:Object Reference) {is=title K2Node_FunctionEntry}
   > n3 = Sequence!(execute) -> (then_0, then_1)
   > n4 = getListItems(self=Get listFilters())
   > n5 = For Each Loop(Array:Array of Object References=@n4) -> (Array Element:Object Reference)
   > n6 = Cast To TA_FilterTab_Button(Object=@n5)
   > Branch(Condition=getFilterEnabled!(self:TA Filter Tab Button Object Reference=@n6) -> (enabled:bool))
   > n7 = Cast To AVT_FilterCategory_ItemData(Object=Get currentListItemData!(self:TA Base Editor Button Widget Object Reference=@n6) -> (currentListItemData:Object Reference))
   n8 = Find!(TargetMap=Get filterCategoryMap() -> (filterCategoryMap=NewEnumerator0), Key=Get filterCategory(self=@n7) -> (filterCategory=NewEnumerator0)) -> (
    Value_filterCategory_22_#15:AVT_FilterCategory_Enum Enum
    Value_children_20_#16:Set of AVT Base Filter Widget Object References
    ReturnValue
   )
   > n9 = To Array(A=@n8.Value_children_20_#16)
   > n10 = For Each Loop(Array:Array of AVT Base Filter Widget Object References=@n9) -> (Array Element:AVT Base Filter Widget Object Reference)
   > Add!(execute, TargetSet:Set of Widget Object References=Get l_filtersToRun(), NewItem:"Widget Object Reference (by ref)"=@n10) -> (then) {is=title K2Node_CallFunction}
   n11 = To Array(execute=@n3.then_1, A:Set of Widget Object References=Get l_filtersToRun()) -> (Result:Array of Widget Object References)
   > n12 = For Each Loop(Array:Array of Widget Object References=@n11) -> (Array Element:Widget Object Reference)
   Return Node!(execute=@n12.Completed, errors:Array of AVT Error Struct Structures=Get l_errors() -> (l_errors:Array of AVT Error Struct Structures))
   > n13 = Cast To AVT_Base_FilterWidget(Object=@n12)
   > n14 = runFilterOnMesh!(execute, self:AVT Base Filter Widget Object Reference=@n13, asset:Object Reference=@n2) -> (then, hasError:bool, Error_Struct:AVT Error Struct Structure)
   > Branch(Condition=@n14)
   > Add(TargetArray:Array of AVT Error Struct Structures=Get l_errors(), NewItem:"AVT Error Struct Structure (by ref)"=@n14.Error_Struct) {is=title K2Node_CallArrayFunction}
   Reroute Node(InputPin:Set of AVT Base Filter Widget Object References=@n8.Value_children_20_#16) -> (OutputPin:Set of AVT Base Filter Widget Object References)
  ]}
 selectMeshInLevel {nodes [
   n1 = selectMeshInLevel!() -> (then, mesh:String) {is=title K2Node_FunctionEntry}
   > findMeshUseInWorld(meshPackageName=@n1) {is=title K2Node_CallFunction}
   > Return Node!(execute, success:bool=true)
  ]}
 setAllFiltersEnabledByCategory {nodes [
   n1 = setAllFiltersEnabledByCategory!() -> (then, enabled:bool=~, filterCategory:AVT_FilterCategory_Enum Enum=NewEnumerator0, isToggle:bool=~) {is=title K2Node_FunctionEntry}
   > Set l_enabled(l_enabled:bool=@n1) -> (Output_Get:bool)
   > n2 = Sequence!(execute) -> (then_0, then_1, then_2:Exec)
   n3 = Get filterCategory!() -> (filterCategory=NewEnumerator0)
   n4 = Get filterCategoryMap() -> (filterCategoryMap=NewEnumerator0)
   > n5 = To Array(A:Set of AVT Base Filter Widget Object References=Break AVT Filter Type Children Struct(AVT_FilterTypeChildren_Struct=Find!(TargetMap=@n4, Key=@n3) -> (Value:AVT Filter Type Children Struct Structure, ReturnValue)).children_20_#16) -> (Result:Array of AVT Base Filter Widget Object References)
   > Set l_filterWidgets(l_filterWidgets=@n5)
   n6 = Get l_filterWidgets()
   n7 = Branch(execute=@n2.then_1, Condition=Is Valid Index(TargetArray=@n6))
   > Branch(Condition=Get isToggle() -> (isToggle:bool))
   n8 = "Get (a copy)"(Array=@n6)
   > Set l_enabled(l_enabled=NOT Boolean(A=getFilterEnabled!(self:AVT Base Filter Widget Object Reference=@n8) -> (isEnabled:bool)))
   Return Node!(execute=@n7.else)
   n9 = For Each Loop(Exec=@n2.then_2, Array:Array of AVT Base Filter Widget Object References=Get l_filterWidgets()) -> (Array Element:AVT Base Filter Widget Object Reference)
   > setFilterEnabled(self=@n9, isEnabled=Get l_enabled() -> (l_enabled:bool))
  ]}
 setCurrentMeshLodSettings {nodes [
   n1 = setCurrentMeshLodSettings!() -> (then, object:Object Reference) {is=title K2Node_FunctionEntry}
   > Set currentMeshForLodSettings(currentMeshForLodSettings:Object Reference=@n1) -> (Output_Get:Object Reference)
   > setAsset!(
    execute
    self:Quick Edit Static Mesh LOD EWidget Object Reference=Get QuickEdit_StaticMesh_LOD_EWidget() -> (QuickEdit_StaticMesh_LOD_EWidget:Quick Edit Static Mesh LOD EWidget Object Reference)
    object:Object Reference=@n1
   ) -> (then)
   > setShow Mesh LODSettings(show=IsValid(Object=@n1))
  ]}
 setShow Mesh LODSettings {nodes [
   n1 = setShow Mesh LODSettings!() -> (then, show:bool=~) {is=title K2Node_FunctionEntry}
   n2 = NOT Boolean(A=Get forceCloseLODSettings())
   > SetVisibility(
    self=[Get TextBlock_2() -> (TextBlock_2:Text Object Reference), Get SizeBox_3()]
    InVisibility=Select(Option 0=Collapsed, Option 1=SelfHitTestInvisible, Index=AND Boolean(A=@n1, B=@n2))
   )
   > validateMeshLODVisuals()
  ]}
 syncBrowserToAsset {nodes [
   n1 = syncBrowserToAsset!() -> (then, asset:String) {is=title K2Node_FunctionEntry}
   > SyncAssetInCB(packageName=@n1) {is=title K2Node_CallFunction}
   > Return Node!(execute, success:bool=false)
  ]}
 syncListDataInCB {nodes [
   n1 = syncListDataInCB!() -> (then, data:Array of Object References) {is=title K2Node_FunctionEntry}
   > n2 = For Each Loop(Array:Array of Object References=@n1) -> (Array Element:Object Reference)
   Branch(execute=@n2.Completed, Condition=Is Valid Index(TargetArray:Array of Strings=Get l_list() -> (l_list:Array of Strings)))
   > SyncBrowserToObjects(AssetPaths=Get l_list())
   n3 = Cast To AVT_ItemData(execute=@n2.LoopBody, Object=@n2)
   > Add(TargetArray:Array of Strings=Get l_list(), NewItem:"String (by ref)"=Get Object Path String(Object=Get(self=@n3)))
  ]}
 validateAssetIsBad {nodes [
   n1 = Inputs!() -> (" ":Exec, asset:Object Reference)
   > Is Valid(InputObject=@n1)
   > n2 = runAssetTests!(execute, self, assetMeshToCheck:Object Reference=@n1) -> (then, errors:Array of AVT Error Struct Structures) {is=title K2Node_CallFunction}
   > n3 = Branch(Condition="integer > integer"(A=Length(TargetArray:Array of AVT Error Struct Structures=@n2) {is=title K2Node_CallArrayFunction}, B=0))
   > Outputs!("bad(AddIt)":Exec, "good(DontAdd)":Exec=@n3.else, outMesh:Static Mesh Object Reference=@n1, errors:Array of AVT Error Struct Structures=@n2)
  ]}
 validateMeshLODVisuals {nodes [
   validateMeshLODVisuals() {is=title K2Node_FunctionEntry}
   n1 = IsVisible(self=Get SizeBox_3())
   > Set buttonColor_Default!(
    execute
    buttonColor_Default:Linear Color Structure=SelectColor(A=(R=1,G=0.0724,B=0.0624,A=1), B=(R=0.098,G=0.451,B=1,A=1), bPickA=@n1)
    self:TA Base Editor Button Widget Object Reference=Get btnToggleClosed() -> (btnToggleClosed:TA Generic 01 Button Object Reference)
   ) -> (then, Output_Get:Linear Color Structure)
   > SetVisibility(
    self=Get txtShrinkLabel() -> (txtShrinkLabel:Text Object Reference)
    InVisibility=Select(Option 0=HitTestInvisible, Option 1=Collapsed, Index=@n1)
   )
   n2 = IsVisible(self=Get SizeBox_3())
   n3 = Get btnToggleClosed()
   > Set iconBrush!(
    execute
    iconBrush:Slate Brush Structure=MakeBrushFromTexture!(
     Texture:Texture 2D Object Reference=Select(
      Option 0:Texture 2D Object Reference=/Game/UI/Foundation/Textures/Buttons/T-Slot-SelectLoot-64
      Option 1:Texture 2D Object Reference=/Game/UI/Foundation/Textures/Buttons/T-Tab-Secondary-Base-32x64
      Index=@n2
     ) -> (ReturnValue:Texture 2D Object Reference)
     Width:int=16
     Height:int=16
    ) -> (ReturnValue:Slate Brush Structure)
    self:TA Generic 01 Button Object Reference=@n3
   ) -> (then, Output_Get:Slate Brush Structure)
   > validateVisuals!(execute, self:TA Base Editor Button Widget Object Reference=@n3) -> (then, " ":bool)
  ]}
 validateSettings {nodes [
   validateSettings() {is=title K2Node_FunctionEntry}
   > n1 = Sequence!(execute) -> (then_0, then_1, then_2, then_3:Exec)
   SetVisibility(
    execute=@n1.then_2
    self=Get vBoxFilters() -> (vBoxFilters:Vertical Box Object Reference)
    InVisibility=Select(Option 0:ESlateVisibility Enum=Collapsed, Option 1:ESlateVisibility Enum=SelfHitTestInvisible, Index=Get setting_FiltersAreShown()) -> (ReturnValue:ESlateVisibility Enum)
   )
   > setLabel!(
    execute
    self:TA Base Editor Button Widget Object Reference=Get btnShowFilters() -> (btnShowFilters:TA Generic 01 Button Object Reference)
    label:String=Select(Option 0:String=Show Filters, Option 1:String=Hide Filters, Index=Get setting_FiltersAreShown()) -> (ReturnValue:String)
   ) -> (then)
   n2 = Get TA_Generic01_List()
   n3 = "To String (Integer)"(InInt=Length!(TargetMap:Map of User Widget Object References to Object References=Get listItemDataMap!(self:TA Generic 01 List Object Reference=@n2) -> (listItemDataMap:Map of User Widget Object References to Object References)) -> (ReturnValue=0) {is=title K2Node_CallFunction})
   "SetText (Text)"(
    execute=@n1.then_3
    self=Get txtListInfo() -> (txtListInfo:Text Object Reference)
    InText="To Text (String)"(InString=Append(A="Showing ", B=@n3, C=" Assets"))
   )
  ]}
]
variables [
 currentMeshForLodSettings {type=Object, default=null}
 filterCategoryMap {default {}}
 forceCloseLODSettings {type=boolean, default=true}
 setting_excludedPackageNames {type="array<string>", default [_ImposterMesh, _LootTiered, Fort_Foundation, SkyDome, Box_Pivot]}
 setting_FiltersAreShown {type=boolean}
]
widgets {
 info {
  parentClass=/Game/Blueprints/Blutilities/ForTechArt/Widgets/TA_Base_EditorWidget.TA_Base_EditorWidget_C
  rootWidgetClass=&Overlay
  widgetCount=58
 }
 widgets [
  overlayMain {widgetClassPath=&Overlay, bIsVariable=true, widget=overlayMain}
  HorizontalBox_0 {
   widgetClassPath=&HorizontalBox
   parent=overlayMain
   widget=HorizontalBox_0
   slot {class=OverlaySlot, properties {horizontalAlignment=HAlign_Fill, verticalAlignment=VAlign_Fill}}
  }
  VerticalBox_30 {
   widgetClassPath=&VerticalBox
   parent=HorizontalBox_0
   widget=VerticalBox_30
   slot {class=HorizontalBoxSlot, properties {size {sizeRule=Fill, value=1}}}
  }
  TextBlock {
   widgetClassPath=&TextBlock
   parent=VerticalBox_30
   widget=TextBlock
   properties {
    autoWrapText=true
    justification=Center
    text=THIS TOOL FINDS ASSETS THAT MATCH THE ACTIVATED FILTERS BELOW.
    font {
     bForceMonospaced=false
     bMaterialIsStencil=false
     fontMaterial=null
     fontObject=&Roboto
     letterSpacing=0
     monospacedWidth=1
     size=10
     skewAmount=0
     typefaceFontName=Bold
     outlineSettings {
      bApplyOutlineToDropShadows=false
      bMiteredCorners=false
      bSeparateFillAlpha=false
      outlineMaterial=null
      outlineSize=0
      outlineColor=<rgba 0 0 0 1>
     }
    }
   }
   slot {class=VerticalBoxSlot, properties {verticalAlignment=VAlign_Center}}
  }
  btnShowFilters {
   widgetClassPath=&TA_Generic01_Button
   parent=VerticalBox_30
   widget=btnShowFilters
   properties {
    label=Show Filters
    toolTipText="Toggles the filter settings.  Edit these to change requirements for meshes to fill the list."
    buttonColor_Default=<rgba 1 0.7408 0.1979 1>
    labelColor_Default=<rgba 1 0.987 0.9767 1>
   }
   slot {class=VerticalBoxSlot, properties {padding=<ltrb 10 0 10 0>}}
  }
  vBoxFilters {
   widgetClassPath=&VerticalBox
   bIsVariable=true
   parent=VerticalBox_30
   widget=vBoxFilters
   properties {visibility=Collapsed}
   slot {class=VerticalBoxSlot, properties {padding=<ltrb 10 0 10 0>}}
  }
  borderFilters {
   widgetClassPath=&Border
   parent=vBoxFilters
   widget=borderFilters
   properties {brushColor=<rgba 0.0573 0.0374 0.0269 1>, padding=<ltrb 5 0 5 5>}
   slot {class=VerticalBoxSlot, properties {padding=<ltrb 10 0 10 0>}}
  }
  VerticalBox_0 {
   widgetClassPath=&VerticalBox
   parent=borderFilters
   widget=VerticalBox_0
   slot {class=BorderSlot, properties {padding=<ltrb 5 0 5 5>}}
  }
  HorizontalBox_2 {
   widgetClassPath=&HorizontalBox
   parent=VerticalBox_0
   widget=HorizontalBox_2
   slot {class=VerticalBoxSlot, properties {padding=<ltrb 0 5 0 0>}}
  }
  HelpHoverWidget_EWidget_1 {
   widgetClassPath=/Game/Blueprints/Blutilities/ForTechArt/Widgets/Misc/HelpHoverWidget_EWidget
   bIsVariable=true
   parent=HorizontalBox_2
   widget=HelpHoverWidget_EWidget_1
   properties {toolTipText="Click a tab to view filters in that Category\r\nClick the checkbox to enable/disable all filters in the Category\r\nCtrl + Click a checkbox to toggle the enabled state of all filters inside a category.\r\nShift + Click a checkbox to toggle all Categories enabled. (Does not affect the filters inside them)"}
   slot {class=HorizontalBoxSlot, properties {horizontalAlignment=HAlign_Left, verticalAlignment=VAlign_Center, padding=<ltrb 0 3 5 3>}}
  }
  listFilters {
   widgetClassPath=/Game/Blueprints/Blutilities/ForTechArt/Widgets/List/TA_Generic01_List
   bIsVariable=true
   parent=HorizontalBox_2
   widget=listFilters
   properties {
    allowMultiSelect=false
    listItemClass=/Game/Blueprints/Blutilities/ForTechArt/Widgets/Buttons/TA_FilterTab_Button.TA_FilterTab_Button_C
    overrideColor=true
    scrollOrientation=Orient_Horizontal
    buttonColor_Default=<rgba 0.2969 0.2214 0.0742 1>
    buttonColor_Selected=<rgba 0.7552 0.5595 0.1495 1>
    labelColor_Default=<rgba 0.9427 0.9427 0.9427 1>
    labelColor_Selected=<rgba 1 1 1 1>
   }
   slot {class=HorizontalBoxSlot, properties {verticalAlignment=VAlign_Bottom}}
  }
  Image_11 {
   widgetClassPath=&Image
   bIsVariable=true
   parent=VerticalBox_0
   widget=Image_11
   properties {brush {
     bIsDynamicallyLoaded=false
     drawAs=Image
     imageType=NoImage
     mirroring=NoMirror
     resourceName=None
     resourceObject=null
     tiling=NoTile
     imageSize=<32 4>
     margin=<ltrb 0 0 0 0>
     outlineSettings {
      bUseBrushTransparency=false
      roundingType=HalfHeightRadius
      width=0
      color {colorUseRule=UseColor_Specified, specifiedColor=<rgba 0 0 0 0>}
      cornerRadii=<0 0 0 0>
     }
     tintColor {colorUseRule=UseColor_Specified, specifiedColor=<rgba 1 0.7042 0.0984 1>}
     uVRegion {bIsValid=false, max=<0 0>, min=<0 0>}
    }}
   slot {class=VerticalBoxSlot, properties {verticalAlignment=VAlign_Bottom}}
  }
  Border_1 {
   widgetClassPath=&Border
   parent=VerticalBox_0
   widget=Border_1
   properties {brushColor=<rgba 0.8021 0.4636 0.1818 0.335>, padding=<ltrb 5 5 5 5>}
   slot {class=VerticalBoxSlot}
  }
  wrapBoxFilters {
   widgetClassPath=&WrapBox
   bIsVariable=true
   parent=Border_1
   widget=wrapBoxFilters
   properties {innerSlotPadding=<5 5>}
   slot {class=BorderSlot, properties {padding=<ltrb 5 5 5 5>}}
  }
  btnResetFilters {
   widgetClassPath=&TA_Generic01_Button
   parent=wrapBoxFilters
   widget=btnResetFilters
   properties {
    toolTipText=Toggle all visible filters on/off
    buttonColor_Default=<rgba 1 0.0469 0.0469 1>
    iconBrush {
     bIsDynamicallyLoaded=false
     drawAs=Image
     imageType=NoImage
     mirroring=NoMirror
     resourceName=None
     resourceObject=/Game/Athena/HUD/Art/BacchusHudIcons/Swap_Normal
     tiling=NoTile
     imageSize=<16 16>
     margin=<ltrb 0 0 0 0>
     outlineSettings {
      bUseBrushTransparency=false
      roundingType=HalfHeightRadius
      width=0
      color {colorUseRule=UseColor_Specified, specifiedColor=<rgba 0 0 0 0>}
      cornerRadii=<0 0 0 1>
     }
     tintColor {colorUseRule=UseColor_Specified, specifiedColor=<rgba 1 1 1 1>}
     uVRegion {bIsValid=false, max=<0 0>, min=<0 0>}
    }
    labelColor_Default=<rgba 0.0131 0.0156 0.0137 1>
   }
   slot {class=WrapBoxSlot, properties {horizontalAlignment=HAlign_Center, verticalAlignment=VAlign_Center}}
  }
  AVT_SM_LODsTooSimilar_FilterWidget {
   widgetClassPath=/Game/Blueprints/Blutilities/ForArtists/AssetValidationTool/SubWidgets/FilterWidgets/AVT_SM_LODsTooSimilar_FilterWidget
   bIsVariable=true
   parent=wrapBoxFilters
   widget=AVT_SM_LODsTooSimilar_FilterWidget
   slot {class=WrapBoxSlot}
  }
  SMV_HasLODs_FilterWidget {
   widgetClassPath=/Game/Blueprints/Blutilities/ForArtists/AssetValidationTool/SubWidgets/FilterWidgets/AVT_SM_HasLODs_FilterWidget
   bIsVariable=true
   parent=wrapBoxFilters
   widget=SMV_HasLODs_FilterWidget
   slot {class=WrapBoxSlot}
  }
  SMV_NoMobileMin_FilterWidget {
   widgetClassPath=/Game/Blueprints/Blutilities/ForArtists/AssetValidationTool/SubWidgets/FilterWidgets/AVT_SM_NoMobileMin_FilterWidget
   bIsVariable=true
   parent=wrapBoxFilters
   widget=SMV_NoMobileMin_FilterWidget
   slot {class=WrapBoxSlot}
  }
  SMV_ComplexLOD_FilterWidget {
   widgetClassPath=/Game/Blueprints/Blutilities/ForArtists/AssetValidationTool/SubWidgets/FilterWidgets/AVT_SM_ComplexLOD_FilterWidget
   bIsVariable=true
   parent=wrapBoxFilters
   widget=SMV_ComplexLOD_FilterWidget
   slot {class=WrapBoxSlot}
  }
  SMV_TooManyVerts_FilterWidget {
   widgetClassPath=/Game/Blueprints/Blutilities/ForArtists/AssetValidationTool/SubWidgets/FilterWidgets/AVT_SM_TooManyVerts_FilterWidget
   bIsVariable=true
   parent=wrapBoxFilters
   widget=SMV_TooManyVerts_FilterWidget
   slot {class=WrapBoxSlot}
  }
  SMV_TooManyMats_FilterWidget {
   widgetClassPath=/Game/Blueprints/Blutilities/ForArtists/AssetValidationTool/SubWidgets/FilterWidgets/AVT_SM_TooManyMats_FilterWidget
   bIsVariable=true
   parent=wrapBoxFilters
   widget=SMV_TooManyMats_FilterWidget
   slot {class=WrapBoxSlot}
  }
  AVT_Tex_TooLarge_FilterWidget {
   widgetClassPath=/Game/Blueprints/Blutilities/ForArtists/AssetValidationTool/SubWidgets/FilterWidgets/AVT_Tex_TooLarge_FilterWidget
   bIsVariable=true
   parent=wrapBoxFilters
   widget=AVT_Tex_TooLarge_FilterWidget
   slot {class=WrapBoxSlot}
  }
  AVT_Tex_NotStreaming_FilterWidget {
   widgetClassPath=/Game/Blueprints/Blutilities/ForArtists/AssetValidationTool/SubWidgets/FilterWidgets/AVT_Tex_NotStreaming_FilterWidget
   bIsVariable=true
   parent=wrapBoxFilters
   widget=AVT_Tex_NotStreaming_FilterWidget
   slot {class=WrapBoxSlot}
  }
  AVT_SK_NoLODSettings_FilterWidget {
   widgetClassPath=/Game/Blueprints/Blutilities/ForArtists/AssetValidationTool/SubWidgets/FilterWidgets/AVT_SK_NoLODSettings_FilterWidget
   bIsVariable=true
   parent=wrapBoxFilters
   widget=AVT_SK_NoLODSettings_FilterWidget
   slot {class=WrapBoxSlot}
  }
  AVT_SK_HasCloth_FilterWidget {
   widgetClassPath=/Game/Blueprints/Blutilities/ForArtists/AssetValidationTool/SubWidgets/FilterWidgets/AVT_SK_HasCloth_FilterWidget
   bIsVariable=true
   parent=wrapBoxFilters
   widget=AVT_SK_HasCloth_FilterWidget
   slot {class=WrapBoxSlot}
  }
  AVT_SM_ManyCollPrims_FilterWidget {
   widgetClassPath=/Game/Blueprints/Blutilities/ForArtists/AssetValidationTool/SubWidgets/FilterWidgets/AVT_SM_ManyCollPrims_FilterWidget
   bIsVariable=true
   parent=wrapBoxFilters
   widget=AVT_SM_ManyCollPrims_FilterWidget
   slot {class=WrapBoxSlot}
  }
  AVT_SM_DFScale_FilterWidget {
   widgetClassPath=/Game/Blueprints/Blutilities/ForArtists/AssetValidationTool/SubWidgets/FilterWidgets/AVT_SM_DFScale_FilterWidget
   bIsVariable=true
   parent=wrapBoxFilters
   widget=AVT_SM_DFScale_FilterWidget
   slot {class=WrapBoxSlot}
  }
  AVT_SM_NoQualityMinLOD_FilterWidget {
   widgetClassPath=/Game/Blueprints/Blutilities/ForArtists/AssetValidationTool/SubWidgets/FilterWidgets/AVT_SM_NoQualityMinLOD_FilterWidget
   bIsVariable=true
   parent=wrapBoxFilters
   widget=AVT_SM_NoQualityMinLOD_FilterWidget
   slot {class=WrapBoxSlot}
  }
  Border_0 {
   widgetClassPath=&Border
   parent=VerticalBox_30
   widget=Border_0
   properties {brushColor=<rgba 0.0482 0.0482 0.0482 1>, padding=<ltrb 10 5 10 10>}
   slot {class=VerticalBoxSlot, properties {padding=<ltrb 10 5 10 10>, size {sizeRule=Fill, value=1}}}
  }
  VerticalBox_32 {
   widgetClassPath=&VerticalBox
   parent=Border_0
   widget=VerticalBox_32
   slot {class=BorderSlot, properties {padding=<ltrb 10 5 10 10>}}
  }
  TextBlock_1 {
   widgetClassPath=&TextBlock
   parent=VerticalBox_32
   widget=TextBlock_1
   properties {
    autoWrapText=true
    justification=Center
    text=CLICK A BUTTON TO FIND BAD ASSETS:
    font {
     bForceMonospaced=false
     bMaterialIsStencil=false
     fontMaterial=null
     fontObject=&Roboto
     letterSpacing=0
     monospacedWidth=1
     size=10
     skewAmount=0
     typefaceFontName=Bold
     outlineSettings {
      bApplyOutlineToDropShadows=false
      bMiteredCorners=false
      bSeparateFillAlpha=false
      outlineMaterial=null
      outlineSize=0
      outlineColor=<rgba 0 0 0 1>
     }
    }
   }
   slot {class=VerticalBoxSlot, properties {verticalAlignment=VAlign_Center}}
  }
  HorizontalBox_8 {widgetClassPath=&HorizontalBox, parent=VerticalBox_32, widget=HorizontalBox_8, slot {class=VerticalBoxSlot}}
  btnFindBadAssets_InLevel {
   widgetClassPath=&TA_Generic01_Button
   parent=HorizontalBox_8
   widget=btnFindBadAssets_InLevel
   properties {label=In Level, toolTipText=Fill the list with meshes from the current level that match the filter.}
   slot {class=HorizontalBoxSlot, properties {verticalAlignment=VAlign_Center, size {sizeRule=Fill, value=1}}}
  }
  btnFindBadAsset_InCB {
   widgetClassPath=&TA_Generic01_Button
   parent=HorizontalBox_8
   widget=btnFindBadAsset_InCB
   properties {label=In Content Browser Selection, toolTipText=Find assets selected in the content browser that match the filter.}
   slot {class=HorizontalBoxSlot, properties {verticalAlignment=VAlign_Center, padding=<ltrb 10 0 0 0>, size {sizeRule=Fill, value=1}}}
  }
  HorizontalBox_9 {
   widgetClassPath=&HorizontalBox
   parent=VerticalBox_32
   widget=HorizontalBox_9
   slot {class=VerticalBoxSlot, properties {padding=<ltrb 0 5 0 0>}}
  }
  HelpHoverWidget_EWidget {
   widgetClassPath=&HelpHoverWidget_EWidget
   bIsVariable=true
   parent=HorizontalBox_9
   widget=HelpHoverWidget_EWidget
   properties {toolTipText="CTRL + Left Click = Find in Browser\r\n Shift + Left Click = to Open Asset\r\n           Right Click = Find Actors in Level"}
   slot {
    class=HorizontalBoxSlot
    properties {horizontalAlignment=HAlign_Left, verticalAlignment=VAlign_Center, padding=<ltrb 0 3 0 3>, size {sizeRule=Fill, value=1}}
   }
  }
  Border_7 {
   widgetClassPath=&Border
   parent=HorizontalBox_9
   widget=Border_7
   properties {horizontalAlignment=HAlign_Center, verticalAlignment=VAlign_Center, brushColor=<rgba 0.0137 0.0137 0.0137 1>, padding=<ltrb 5 2 5 2>}
   slot {class=HorizontalBoxSlot, properties {horizontalAlignment=HAlign_Center}}
  }
  txtListInfo {
   widgetClassPath=&TextBlock
   bIsVariable=true
   parent=Border_7
   widget=txtListInfo
   properties {
    justification=Center
    colorAndOpacity {colorUseRule=UseColor_Specified, specifiedColor=<rgba 1 1 1 0.5>}
    font {
     bForceMonospaced=false
     bMaterialIsStencil=false
     fontMaterial=null
     fontObject=&Roboto
     letterSpacing=0
     monospacedWidth=1
     size=8
     skewAmount=0
     typefaceFontName=Bold
     outlineSettings {
      bApplyOutlineToDropShadows=false
      bMiteredCorners=false
      bSeparateFillAlpha=false
      outlineMaterial=null
      outlineSize=0
      outlineColor=<rgba 0 0 0 1>
     }
    }
   }
   slot {class=BorderSlot, properties {horizontalAlignment=HAlign_Center, verticalAlignment=VAlign_Center, padding=<ltrb 5 2 5 2>}}
  }
  btnRemoveAllFromList {
   widgetClassPath=&TA_Generic01_Button
   parent=HorizontalBox_9
   widget=btnRemoveAllFromList
   properties {
    label=CLEAR LIST
    toolTipText=Clear all assets from the list.
    buttonColor_Default=<rgba 1 1 1 1>
    buttonStyle_Default {
     clickedSlateSound {resourceObject=null}
     disabled {
      bIsDynamicallyLoaded=false
      drawAs=Box
      imageType=NoImage
      mirroring=NoMirror
      resourceName=None
      resourceObject=/Game/Blueprints/Blutilities/ForTechArt/ArtAssets/Textures/Button02/MI_TA_ButtonBG_02_Down
      tiling=NoTile
      imageSize=<32 32>
      margin=<ltrb 0.25 0.25 0.25 0.25>
      outlineSettings {
       bUseBrushTransparency=false
       roundingType=HalfHeightRadius
       width=0
       color {colorUseRule=UseColor_Specified, specifiedColor=<rgba 0 0 0 0>}
       cornerRadii=<0 0 0 1>
      }
      tintColor {colorUseRule=UseColor_Specified, specifiedColor=<rgba 0.3333 0.3333 0.3333 1>}
      uVRegion {bIsValid=false, max=<0 0>, min=<0 0>}
     }
     disabledForeground {colorUseRule=UseColor_Foreground, specifiedColor=<rgba 1 0 1 1>}
     hovered {
      bIsDynamicallyLoaded=false
      drawAs=Box
      imageType=NoImage
      mirroring=NoMirror
      resourceName=None
      resourceObject=/Game/Blueprints/Blutilities/ForTechArt/ArtAssets/Textures/Button02/MI_TA_ButtonBG_02_Over
      tiling=NoTile
      imageSize=<32 32>
      margin=<ltrb 0.25 0.25 0.25 0.25>
      outlineSettings {
       bUseBrushTransparency=false
       roundingType=HalfHeightRadius
       width=0
       color {colorUseRule=UseColor_Specified, specifiedColor=<rgba 0 0 0 0>}
       cornerRadii=<0 0 0 1>
      }
      tintColor {colorUseRule=UseColor_Specified, specifiedColor=<rgba 0.8542 0 0 1>}
      uVRegion {bIsValid=false, max=<0 0>, min=<0 0>}
     }
     hoveredForeground {colorUseRule=UseColor_Foreground, specifiedColor=<rgba 1 0 1 1>}
     hoveredSlateSound {resourceObject=null}
     normal {
      bIsDynamicallyLoaded=false
      drawAs=Box
      imageType=NoImage
      mirroring=NoMirror
      resourceName=None
      resourceObject=/Game/Blueprints/Blutilities/ForTechArt/ArtAssets/Textures/Button02/M_TA_ButtonBG_02_Up
      tiling=NoTile
      imageSize=<32 32>
      margin=<ltrb 0.25 0.25 0.25 0.25>
      outlineSettings {
       bUseBrushTransparency=false
       roundingType=HalfHeightRadius
       width=0
       color {colorUseRule=UseColor_Specified, specifiedColor=<rgba 0 0 0 0>}
       cornerRadii=<0 0 0 1>
      }
      tintColor {colorUseRule=UseColor_Specified, specifiedColor=<rgba 0.5457 0 0 1>}
      uVRegion {bIsValid=false, max=<0 0>, min=<0 0>}
     }
     normalForeground {colorUseRule=UseColor_Foreground, specifiedColor=<rgba 1 0 1 1>}
     normalPadding=<ltrb 0 0 0 0>
     pressed {
      bIsDynamicallyLoaded=false
      drawAs=Box
      imageType=NoImage
      mirroring=NoMirror
      resourceName=None
      resourceObject=&MI_TA_ButtonBG_02_Down
      tiling=NoTile
      imageSize=<32 32>
      margin=<ltrb 0.25 0.25 0.25 0.25>
      outlineSettings {
       bUseBrushTransparency=false
       roundingType=HalfHeightRadius
       width=0
       color {colorUseRule=UseColor_Specified, specifiedColor=<rgba 0 0 0 0>}
       cornerRadii=<0 0 0 1>
      }
      tintColor {colorUseRule=UseColor_Specified, specifiedColor=<rgba 0.2396 0 0 1>}
      uVRegion {bIsValid=false, max=<0 0>, min=<0 0>}
     }
     pressedForeground {colorUseRule=UseColor_Foreground, specifiedColor=<rgba 1 0 1 1>}
     pressedPadding=<ltrb 0 0 0 0>
     pressedSlateSound {resourceObject=null}
    }
    labelColor_Default=<rgba 0.9219 0.9219 0.9219 1>
   }
   slot {
    class=HorizontalBoxSlot
    properties {horizontalAlignment=HAlign_Right, verticalAlignment=VAlign_Center, padding=<ltrb 10 0 0 0>, size {sizeRule=Fill, value=1}}
   }
  }
  Border_5 {
   widgetClassPath=&Border
   parent=VerticalBox_32
   widget=Border_5
   properties {brushColor=<rgba 0 0 0 0.5>, padding=<ltrb 4 4 4 4>}
   slot {class=VerticalBoxSlot, properties {size {sizeRule=Fill, value=1}}}
  }
  TA_Generic01_List {
   widgetClassPath=&TA_Generic01_List
   bIsVariable=true
   parent=Border_5
   widget=TA_Generic01_List
   properties {
    allowMultiSelect=false
    itemPadding=0
    listItemClass=/Game/Blueprints/Blutilities/ForArtists/AssetValidationTool/SubWidgets/AVT_Asset_ListItemWidget.AVT_Asset_ListItemWidget_C
   }
   slot {class=BorderSlot, properties {padding=<ltrb 4 4 4 4>}}
  }
  btnPrintToLog {
   widgetClassPath=&TA_Generic01_Button
   parent=VerticalBox_32
   widget=btnPrintToLog
   properties {label=Print List to Log, buttonColor_Default=<rgba 0.6826 1 0.505 1>, labelColor_Default=<rgba 0.0131 0.0156 0.0137 1>}
   slot {class=VerticalBoxSlot, properties {padding=<ltrb 0 10 0 0>}}
  }
  HorizontalBox_7 {
   widgetClassPath=&HorizontalBox
   parent=VerticalBox_32
   widget=HorizontalBox_7
   slot {class=VerticalBoxSlot, properties {padding=<ltrb 0 10 0 0>}}
  }
  btnExclude_Selection {
   widgetClassPath=&TA_Generic01_Button
   parent=HorizontalBox_7
   widget=btnExclude_Selection
   properties {
    label=Add List Selection To Exclusion List
    toolTipText=Add selected list items to a list of assets that will never show in the list again.
    buttonColor_Default=<rgba 1 0.2813 0.0398 1>
    buttonColor_Selected=<rgba 1 0.8721 0 1>
    labelColor_Default=<rgba 0.9219 0.9219 0.9219 1>
   }
   slot {class=HorizontalBoxSlot, properties {size {sizeRule=Fill, value=1}}}
  }
  btnExclusion_Edit {
   widgetClassPath=&TA_Generic01_Button
   parent=HorizontalBox_7
   widget=btnExclusion_Edit
   properties {label=Edit Exclusion List, toolTipText=Open the Exclusion Edit Popup to edit your exclusion list}
   slot {class=HorizontalBoxSlot, properties {padding=<ltrb 10 0 0 0>}}
  }
  btnSelectAllInCB {
   widgetClassPath=&TA_Generic01_Button
   parent=VerticalBox_32
   widget=btnSelectAllInCB
   properties {
    label=Select Current List in Content Browser
    toolTipText="Select all list assets in the content browser.  From there you can make collections/bulk edit."
   }
   slot {class=VerticalBoxSlot, properties {padding=<ltrb 0 10 0 0>}}
  }
  borderLodSettings {
   widgetClassPath=&Border
   bIsVariable=true
   parent=HorizontalBox_0
   widget=borderLodSettings
   properties {visibility=Collapsed, brushColor=<rgba 0.026 0.026 0.026 1>, padding=<ltrb 0 0 0 0>}
   slot {class=HorizontalBoxSlot}
  }
  VerticalBox_1 {
   widgetClassPath=&VerticalBox
   parent=borderLodSettings
   widget=VerticalBox_1
   slot {class=BorderSlot, properties {padding=<ltrb 0 0 0 0>}}
  }
  Border_133 {
   widgetClassPath=&Border
   parent=VerticalBox_1
   widget=Border_133
   properties {brushColor=<rgba 0.0104 0.0104 0.0104 1>}
   slot {class=VerticalBoxSlot}
  }
  HorizontalBox_1 {widgetClassPath=&HorizontalBox, parent=Border_133, widget=HorizontalBox_1, slot {class=BorderSlot}}
  TextBlock_2 {
   widgetClassPath=&TextBlock
   bIsVariable=true
   parent=HorizontalBox_1
   widget=TextBlock_2
   properties {
    justification=Center
    text=Mesh LOD Settings
    colorAndOpacity {colorUseRule=UseColor_Specified, specifiedColor=<rgba 1 1 1 0.5>}
    font {
     bForceMonospaced=false
     bMaterialIsStencil=false
     fontMaterial=null
     fontObject=&Roboto
     letterSpacing=0
     monospacedWidth=1
     size=10
     skewAmount=0
     typefaceFontName=Bold
     outlineSettings {
      bApplyOutlineToDropShadows=false
      bMiteredCorners=false
      bSeparateFillAlpha=false
      outlineMaterial=null
      outlineSize=0
      outlineColor=<rgba 0 0 0 1>
     }
    }
   }
   slot {class=HorizontalBoxSlot, properties {horizontalAlignment=HAlign_Center, verticalAlignment=VAlign_Center, size {sizeRule=Fill, value=1}}}
  }
  btnToggleClosed {
   widgetClassPath=&TA_Generic01_Button
   parent=HorizontalBox_1
   widget=btnToggleClosed
   properties {iconBrush {
     bIsDynamicallyLoaded=false
     drawAs=Image
     imageType=NoImage
     mirroring=NoMirror
     resourceName=None
     resourceObject=&T-Slot-SelectLoot-64
     tiling=NoTile
     imageSize=<16 16>
     margin=<ltrb 0 0 0 0>
     outlineSettings {
      bUseBrushTransparency=false
      roundingType=HalfHeightRadius
      width=0
      color {colorUseRule=UseColor_Specified, specifiedColor=<rgba 0 0 0 0>}
      cornerRadii=<0 0 0 1>
     }
     tintColor {colorUseRule=UseColor_Specified, specifiedColor=<rgba 1 1 1 1>}
     uVRegion {bIsValid=false, max=<0 0>, min=<0 0>}
    }}
   slot {class=HorizontalBoxSlot}
  }
  Overlay_1 {
   widgetClassPath=&Overlay
   parent=VerticalBox_1
   widget=Overlay_1
   slot {class=VerticalBoxSlot, properties {size {sizeRule=Fill, value=1}}}
  }
  SizeBox_3 {
   widgetClassPath=&SizeBox
   bIsVariable=true
   parent=Overlay_1
   widget=SizeBox_3
   properties {bOverride_WidthOverride=true, widthOverride=370}
   slot {class=OverlaySlot, properties {verticalAlignment=VAlign_Fill}}
  }
  QuickEdit_StaticMesh_LOD_EWidget {
   widgetClassPath=/Game/Blueprints/Blutilities/ForArtists/AssetQuickEditMenus/QuickEdit_StaticMesh_LOD_EWidget
   bIsVariable=true
   parent=SizeBox_3
   widget=QuickEdit_StaticMesh_LOD_EWidget
   slot {class=SizeBoxSlot}
  }
  SizeBox_5 {
   widgetClassPath=&SizeBox
   parent=Overlay_1
   widget=SizeBox_5
   properties {bOverride_MaxDesiredWidth=true}
   slot {class=OverlaySlot, properties {horizontalAlignment=HAlign_Center, verticalAlignment=VAlign_Center}}
  }
  txtShrinkLabel {
   widgetClassPath=&TextBlock
   bIsVariable=true
   parent=SizeBox_5
   widget=txtShrinkLabel
   properties {
    justification=Center
    text=Mesh LOD Settings
    visibility=HitTestInvisible
    colorAndOpacity {colorUseRule=UseColor_Specified, specifiedColor=<rgba 1 1 1 0.5>}
    font {
     bForceMonospaced=false
     bMaterialIsStencil=false
     fontMaterial=null
     fontObject=&Roboto
     letterSpacing=0
     monospacedWidth=1
     size=10
     skewAmount=0
     typefaceFontName=Bold
     outlineSettings {
      bApplyOutlineToDropShadows=false
      bMiteredCorners=false
      bSeparateFillAlpha=false
      outlineMaterial=null
      outlineSize=0
      outlineColor=<rgba 0 0 0 1>
     }
    }
    renderTransform {angle=-90, scale=<1 1>, shear=<0 0>, translation=<0 0>}
   }
   slot {class=SizeBoxSlot, properties {horizontalAlignment=HAlign_Center}}
  }
  StaticMeshValidator_ExclusionList_Widget {
   widgetClassPath=/Game/Blueprints/Blutilities/ForArtists/AssetValidationTool/SubWidgets/AVT_ExclusionList_Widget
   bIsVariable=true
   parent=overlayMain
   widget=StaticMeshValidator_ExclusionList_Widget
   properties {visibility=Collapsed}
   slot {class=OverlaySlot, properties {horizontalAlignment=HAlign_Fill, verticalAlignment=VAlign_Fill}}
  }
 ]
}
