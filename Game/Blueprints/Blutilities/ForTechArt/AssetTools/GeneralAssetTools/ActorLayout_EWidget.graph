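# Editor Utility Widget that arranges or offsets selected level actors, wrapping actions in undo transactions.
# Line layout chains actors along X from the first, using bounds and padding.
# Grid layout builds a square-root grid from world origin at Z=0, sized by largest bounds or an override, correcting pivot offsets; three or fewer actors use line layout.
# Offset applies spin-box XYZ values in world or local space.
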EditorUtilityWidgetBlueprint /Game/Blueprints/Blutilities/ForTechArt/AssetTools/GeneralAssetTools/ActorLayout_EWidget
graphs [
 EventGraph {nodes [
   "On Clicked (btnLayoutGrid)"!() -> (OutputDelegate, then, widget:User Widget Object Reference, event:Pointer Event Structure) {is=title K2Node_ComponentBoundEvent}
   > BeginTransaction(Context=Actor Layout Widget, Description="INVTEXT(\"Layout Actors in Grid\")")
   n1 = Get EditorActorSubsystem()
   > n2 = GetSelectedLevelActors(self=@n1)
   > layoutActors_UniformGrid!(execute, self, actorsToLayout:Array of Actor Object References=@n2) -> (then)
   > EndTransaction()
   "On Clicked (btnLayoutLine)"!() -> (OutputDelegate, then, widget:User Widget Object Reference, event:Pointer Event Structure) {is=title K2Node_ComponentBoundEvent}
   > BeginTransaction(Context=Actor Layout Widget, Description="INVTEXT(\"Layout Actors in Line\")")
   > n3 = GetSelectedLevelActors(self=@n1)
   > layoutActors_Line!(execute, self, actorsToLayout:Array of Actor Object References=@n3) -> (then)
   > EndTransaction()
   "On Checked State Changed (detailGridUseBounds)"!() -> (OutputDelegate, then, isChecked:bool) {is=title K2Node_ComponentBoundEvent}
   > validateGridWidgets()
   n4 = "On Float Value Committed (detailSpacing_3)"!() -> (OutputDelegate, then, InValue:double, CommitMethod:ETextCommit Enum=Default) {is=title K2Node_ComponentBoundEvent}
   > Set actorPadding(actorPadding:double=@n4.InValue) -> (Output_Get:double)
   "On Clicked (btnAddLocation)"!() -> (OutputDelegate, then, widget:User Widget Object Reference, event:Pointer Event Structure) {is=title K2Node_ComponentBoundEvent}
   > BeginTransaction(Context=Actor Layout Widget, Description="INVTEXT(\"Offset Actor Locations\")")
   > n5 = GetSelectedLevelActors(self=Get EditorActorSubsystem())
   > offsetActors!(execute, self, actorsToLayout:Array of Actor Object References=@n5) -> (then)
   > EndTransaction()
   Event Construct()
   > setFloat!(
    execute
    self:Labeled Entry Widget Object Reference=Get detailSpacing_3() -> (detailSpacing_3:Labeled Entry Widget Object Reference)
    float:double=Get actorPadding() -> (actorPadding:double)
   ) -> (then)
   > validateGridWidgets()
  ]}
 layoutActors_Line {nodes [
   layoutActors_Line!() -> (then, actorsToLayout:Array of Actor Object References) {is=title K2Node_FunctionEntry}
   > BeginTransaction(Context=Actor Layout Widget, Description="LOCTEXT(\"Laying out actors in a line\")")
   > n1 = For Each Loop(Array:Array of Actor Object References=Get actorsToLayout() -> (actorsToLayout:Array of Actor Object References)) -> (Array Element:Actor Object Reference)
   > n2 = Branch(Condition="Not Equal (Integer)"!(A:int=@n1.Array Index, B:int=0) -> (ReturnValue:bool) {is=title K2Node_CallFunction})
   > n3 = Sequence()
   n4 = Get Actor Bounds!(self=@n1, bOnlyCollidingComponents=false, bIncludeFromChildActors) -> (Origin, BoxExtent_X:double, BoxExtent_Y:double, BoxExtent_Z:double)
   n5 = vector + vector!(
    A:Vector=Get l_start() -> (l_start:Vector)
    B_X:double=@n4.BoxExtent_X
    B_Y:double=0
    B_Z:double=0
    C:Vector=vector - vector!(A:Vector=Get Actor Location(self=@n1), B:Vector=@n4) -> (ReturnValue:Vector) {is=title K2Node_CallFunction}
   ) -> (ReturnValue_X:double, ReturnValue_Y:double, ReturnValue_Z:double) {is=title K2Node_CommutativeAssociativeBinaryOperator}
   > Set Actor Location!(
    execute
    self:Actor Object Reference=@n1
    NewLocation_X:double=@n5
    NewLocation_Y:double=@n5.ReturnValue_Y
    NewLocation_Z:double=Get!() -> (l_start_X:double, l_start_Y:double, l_start_Z:double).l_start_Z
    bSweep:bool=false
    bTeleport:bool=false
   ) -> (then, SweepHitResult:Hit Result Structure, ReturnValue:bool) {is=title K2Node_CallFunction}
   n6 = Get Actor Bounds(self=@n1, bOnlyCollidingComponents=false)
   n7 = vector + vector!(A=@n6, B_X=float + float(A=@n6.BoxExtent_X, B=Get actorPadding()) {is=title K2Node_CommutativeAssociativeBinaryOperator}, B_Y=0, B_Z=0) -> (ReturnValue_X, ReturnValue_Y, ReturnValue_Z)
   Set!(
    execute=[@n3.then_1, @n2.else]
    l_start_X:double=@n7
    l_start_Y:double=@n7.ReturnValue_Y
    l_start_Z:double=Get Actor Location!(self=@n1) -> (ReturnValue_X:double, ReturnValue_Y:double, ReturnValue_Z:double).ReturnValue_Z
   ) -> (then, Output_Get:Vector)
  ]}
 layoutActors_UniformGrid {nodes [
   layoutActors_UniformGrid!() -> (then, actorsToLayout:Array of Actor Object References) {is=title K2Node_FunctionEntry}
   n1 = Get actorsToLayout()
   > n2 = Branch(Condition="integer > integer"(A=Length(TargetArray=@n1), B=3) {is=title K2Node_CallFunction})
   > BeginTransaction(Context=Actor Layout Widget, Description="LOCTEXT(\"Laying out actors in a line\")")
   > n3 = Sequence!(execute) -> (then_0, then_1, then_2:Exec, then_3:Exec)
   > n4 = Branch(Condition=getBool!(self:Labeled Entry Widget Object Reference=Get detailGridUseBounds() -> (detailGridUseBounds:Labeled Entry Widget Object Reference)) -> (bool:bool))
   > n5 = For Each Loop(Array=Get actorsToLayout())
   Set l_dist(execute=@n5.Completed, l_dist:double=float * float(A=Get l_dist() -> (l_dist:double), B=2)) -> (Output_Get:double)
   n6 = Vector Length(A=Get Actor Bounds!(self=@n5, bOnlyCollidingComponents=false, bIncludeFromChildActors) -> (Origin, BoxExtent:Vector).BoxExtent)
   Branch(execute=@n5.LoopBody, Condition="float > float"(A=@n6, B=Get l_dist()) {is=title K2Node_CallFunction})
   > Set l_dist(l_dist=@n6)
   Set l_dist(
    execute=@n4.else
    l_dist=getFloat!(self:Labeled Entry Widget Object Reference=Get detailGridSize() -> (detailGridSize:Labeled Entry Widget Object Reference)) -> (float:double)
   )
   n7 = "To Float (Integer)"(InInt=Length(TargetArray=Get actorsToLayout()))
   Set l_columnCount(execute=@n3.then_1, l_columnCount:int=Floor(A=Sqrt!(A:double=@n7) -> (ReturnValue:double)) {is=title K2Node_CallFunction}) -> (Output_Get:int=0)
   n8 = For Each Loop(Exec=@n3.then_2, Array=Get actorsToLayout())
   > Set l_currentActor(l_currentActor:Actor Object Reference=@n8) -> (Output_Get:Actor Object Reference)
   n9 = Get l_columnCount() -> (l_columnCount:int=0)
   > Set l_column(l_column:int=int / int!(A:int=@n8.Array Index, B:int=@n9) -> (ReturnValue:int=0) {is=title K2Node_CallFunction}) -> (Output_Get:int=0)
   > Set l_row(l_row:int="% (Integer)"(A=@n8.Array Index, B=@n9) -> (ReturnValue=0)) -> (Output_Get:int=0)
   n10 = Get l_currentActor() -> (l_currentActor:Actor Object Reference)
   n11 = Get l_currentActor()
   > Set l_actorOffsetFromPivot(l_actorOffsetFromPivot:Vector=vector - vector(A=Get Actor Bounds(self=@n11, bOnlyCollidingComponents=false), B=Get Actor Location!(self=@n10) -> (ReturnValue:Vector))) -> (Output_Get:Vector)
   n12 = float + float(A=Get l_dist(), B=Get actorPadding())
   n13 = Get l_row() -> (l_row:int=0)
   n14 = Get l_column() -> (l_column:int=0)
   n15 = vector - vector!(
    A=MakeVector(X=int * float!(A:int=@n14, B:double=@n12) -> (ReturnValue:double), Y=int * float(A=@n13, B=@n12))
    B=Get l_actorOffsetFromPivot() -> (l_actorOffsetFromPivot:Vector)
   ) -> (ReturnValue_X:double, ReturnValue_Y:double, ReturnValue_Z:double)
   > Set Actor Location(self=Get l_currentActor(), NewLocation_X=@n15, NewLocation_Y=@n15.ReturnValue_Y, NewLocation_Z=0, bSweep=false, bTeleport=false)
   layoutActors_Line(execute=@n2.else, actorsToLayout=Get actorsToLayout()) {is=title K2Node_CallFunction}
  ]}
 offsetActors {nodes [
   offsetActors!() -> (then, actorsToLayout:Array of Actor Object References) {is=title K2Node_FunctionEntry}
   > BeginTransaction(Context=Actor Layout Widget, Description="LOCTEXT(\"Laying out actors in a line\")")
   > n1 = For Each Loop(Array=Get actorsToLayout())
   > n2 = Branch(Condition=getBool(self=Get detailOffsetWorld() -> (detailOffsetWorld:Labeled Entry Widget Object Reference)))
   n3 = MakeVector(
    X=GetValue!(self:Spin Box Object Reference=Get SpinBox_x() -> (SpinBox_x:Spin Box Object Reference)) -> (ReturnValue:float)
    Y=GetValue(self=Get SpinBox_y() -> (SpinBox_y:Spin Box Object Reference))
    Z=GetValue(self=Get SpinBox_Z() -> (SpinBox_Z:Spin Box Object Reference))
   )
   > Add Actor World Offset!(execute, self:Actor Object Reference=@n1, DeltaLocation:Vector=@n3, bSweep:bool=false, bTeleport:bool=true) -> (then, SweepHitResult:Hit Result Structure)
   Add Actor Local Offset!(execute=@n2.else, self:Actor Object Reference=@n1, DeltaLocation:Vector=@n3, bSweep:bool=false, bTeleport:bool=false) -> (then, SweepHitResult:Hit Result Structure)
  ]}
 validateGridWidgets {nodes [
   validateGridWidgets() {is=title K2Node_FunctionEntry}
   n1 = Get detailGridUseBounds()
   > SetIsEnabled(self=Get detailGridSize(), bInIsEnabled=NOT Boolean(A=getBool(self=@n1)))
  ]}
]
variables [actorPadding {type=number, default=50}]
widgets {
 info {rootWidgetClass=&VerticalBox, widgetCount=20}
 widgets [
  vertBox {widgetClassPath=&VerticalBox, widget=vertBox}
  CommonTextBlock_2 {
   widgetClassPath=&CommonTextBlock
   parent=vertBox
   widget=CommonTextBlock_2
   properties {
    text=Actor Layout Tool
    font {
     bForceMonospaced=false
     bMaterialIsStencil=false
     fontMaterial=null
     fontObject=&NotoSans
     letterSpacing=0
     monospacedWidth=1
     size=12
     skewAmount=0
     typefaceFontName=NotoSans-Regular
     outlineSettings {
      bApplyOutlineToDropShadows=false
      bMiteredCorners=false
      bSeparateFillAlpha=false
      outlineMaterial=null
      outlineSize=0
      outlineColor=<rgba 0 0 0 1>
     }
    }
    shadowOffset=<0 0>
   }
   slot {class=VerticalBoxSlot, properties {verticalAlignment=VAlign_Center, padding=<ltrb 15 0 0 0>}}
  }
  Image_84 {
   widgetClassPath=&Image
   bIsVariable=true
   parent=vertBox
   widget=Image_84
   properties {
    brush {
     bIsDynamicallyLoaded=false
     drawAs=Image
     imageType=NoImage
     mirroring=NoMirror
     resourceName=None
     resourceObject=null
     tiling=NoTile
     imageSize=<32 2>
     margin=<ltrb 0 0 0 0>
     outlineSettings {
      bUseBrushTransparency=false
      roundingType=HalfHeightRadius
      width=0
      color {colorUseRule=UseColor_Specified, specifiedColor=<rgba 0 0 0 0>}
      cornerRadii=<0 0 0 0>
     }
     tintColor {colorUseRule=UseColor_Specified, specifiedColor=<rgba 1 1 1 1>}
     uVRegion {bIsValid=false, max=<0 0>, min=<0 0>}
    }
    colorAndOpacity=<rgba 0.5312 0.5312 0.5312 1>
   }
   slot {class=VerticalBoxSlot, properties {padding=<ltrb 5 0 5 0>}}
  }
  detailSpacing_3 {
   widgetClassPath=/Game/Blueprints/Blutilities/ForTechArt/Widgets/Misc/labeledEntry_Widget
   bIsVariable=true
   parent=vertBox
   widget=detailSpacing_3
   properties {currentType=float, defaultFloat=5, label="Padding: "}
   slot {class=VerticalBoxSlot, properties {horizontalAlignment=HAlign_Left, verticalAlignment=VAlign_Center, padding=<ltrb 5 5 0 0>}}
  }
  btnLayoutLine {
   widgetClassPath=&TA_Generic01_Button
   parent=vertBox
   widget=btnLayoutLine
   properties {label=Layout Selected Actors in Line}
   slot {class=VerticalBoxSlot, properties {padding=<ltrb 0 10 0 0>}}
  }
  HorizontalBox_0 {
   widgetClassPath=&HorizontalBox
   parent=vertBox
   widget=HorizontalBox_0
   slot {class=VerticalBoxSlot, properties {padding=<ltrb 0 10 0 0>}}
  }
  btnLayoutGrid {
   widgetClassPath=&TA_Generic01_Button
   parent=HorizontalBox_0
   widget=btnLayoutGrid
   properties {label=Layout Selected Actors in Uniform Grid}
   slot {class=HorizontalBoxSlot, properties {verticalAlignment=VAlign_Center, size {sizeRule=Fill, value=1}}}
  }
  ScaleBox_0 {
   widgetClassPath=&ScaleBox
   parent=HorizontalBox_0
   widget=ScaleBox_0
   properties {stretch=UserSpecified, userSpecifiedScale=0.824}
   slot {class=HorizontalBoxSlot, properties {verticalAlignment=VAlign_Center, padding=<ltrb 5 0 0 0>}}
  }
  detailGridUseBounds {
   widgetClassPath=&labeledEntry_Widget
   bIsVariable=true
   parent=ScaleBox_0
   widget=detailGridUseBounds
   properties {defaultBool=true, label=" Use Largest Actor Bounds"}
   slot {class=ScaleBoxSlot}
  }
  ScaleBox {
   widgetClassPath=&ScaleBox
   parent=HorizontalBox_0
   widget=ScaleBox
   properties {stretch=UserSpecified, userSpecifiedScale=0.824}
   slot {class=HorizontalBoxSlot, properties {verticalAlignment=VAlign_Center, padding=<ltrb 5 0 0 0>}}
  }
  detailGridSize {
   widgetClassPath=&labeledEntry_Widget
   bIsVariable=true
   parent=ScaleBox
   widget=detailGridSize
   properties {currentType=float, defaultFloat=100, label=Override Grid Size}
   slot {class=ScaleBoxSlot}
  }
  HorizontalBox {
   widgetClassPath=&HorizontalBox
   parent=vertBox
   widget=HorizontalBox
   slot {class=VerticalBoxSlot, properties {padding=<ltrb 0 10 0 0>}}
  }
  btnAddLocation {
   widgetClassPath=&TA_Generic01_Button
   parent=HorizontalBox
   widget=btnAddLocation
   properties {label=Offset Location of Selected Actors}
   slot {class=HorizontalBoxSlot, properties {verticalAlignment=VAlign_Center, size {sizeRule=Fill, value=1}}}
  }
  txtLabel {
   widgetClassPath=&CommonTextBlock
   parent=HorizontalBox
   widget=txtLabel
   properties {
    text=X:
    font {
     bForceMonospaced=false
     bMaterialIsStencil=false
     fontMaterial=null
     fontObject=&NotoSans
     letterSpacing=0
     monospacedWidth=1
     size=12
     skewAmount=0
     typefaceFontName=NotoSans-Regular
     outlineSettings {
      bApplyOutlineToDropShadows=false
      bMiteredCorners=false
      bSeparateFillAlpha=false
      outlineMaterial=null
      outlineSize=0
      outlineColor=<rgba 0 0 0 1>
     }
    }
    shadowOffset=<0 0>
   }
   slot {class=HorizontalBoxSlot, properties {verticalAlignment=VAlign_Center, padding=<ltrb 5 0 0 0>}}
  }
  SpinBox_x {
   widgetClassPath=/Script/UMG.SpinBox
   bIsVariable=true
   parent=HorizontalBox
   widget=SpinBox_x
   properties {
    foregroundColor {colorUseRule=UseColor_Specified, specifiedColor=<rgba 1 1 1 1>}
    widgetStyle {
     activeBackgroundBrush {
      bIsDynamicallyLoaded=false
      drawAs=RoundedBox
      imageType=NoImage
      mirroring=NoMirror
      resourceName=None
      resourceObject=null
      tiling=NoTile
      imageSize=<0 0>
      margin=<ltrb 0 0 0 0>
      outlineSettings {
       bUseBrushTransparency=false
       roundingType=FixedRadius
       width=1
       color {colorUseRule=UseColor_ColorTable, specifiedColor=<rgba 0 0 0 0>}
       cornerRadii=<4 4 4 4>
      }
      tintColor {colorUseRule=UseColor_Specified, specifiedColor=<rgba 0.0048 0.0048 0.0048 1>}
      uVRegion {bIsValid=false, max=<0 0>, min=<0 0>}
     }
     activeFillBrush {
      bIsDynamicallyLoaded=false
      drawAs=RoundedBox
      imageType=NoImage
      mirroring=NoMirror
      resourceName=None
      resourceObject=null
      tiling=NoTile
      imageSize=<0 0>
      margin=<ltrb 0 0 0 0>
      outlineSettings {
       bUseBrushTransparency=false
       roundingType=FixedRadius
       width=1
       color {colorUseRule=UseColor_Specified, specifiedColor=<rgba 0 0 0 0>}
       cornerRadii=<4 4 4 4>
      }
      tintColor {colorUseRule=UseColor_Specified, specifiedColor=<rgba 0 0.162 0.7454 1>}
      uVRegion {bIsValid=false, max=<0 0>, min=<0 0>}
     }
     arrowsImage {
      bIsDynamicallyLoaded=false
      drawAs=NoDrawType
      imageType=NoImage
      mirroring=NoMirror
      resourceName=None
      resourceObject=null
      tiling=NoTile
      imageSize=<0 0>
      margin=<ltrb 0 0 0 0>
      outlineSettings {
       bUseBrushTransparency=false
       roundingType=HalfHeightRadius
       width=0
       color {colorUseRule=UseColor_Specified, specifiedColor=<rgba 0 0 0 0>}
       cornerRadii=<0 0 0 0>
      }
      tintColor {colorUseRule=UseColor_Specified, specifiedColor=<rgba 1 1 1 1>}
      uVRegion {bIsValid=false, max=<0 0>, min=<0 0>}
     }
     backgroundBrush {
      bIsDynamicallyLoaded=false
      drawAs=RoundedBox
      imageType=NoImage
      mirroring=NoMirror
      resourceName=None
      resourceObject=null
      tiling=NoTile
      imageSize=<0 0>
      margin=<ltrb 0 0 0 0>
      outlineSettings {
       bUseBrushTransparency=false
       roundingType=FixedRadius
       width=1
       color {colorUseRule=UseColor_ColorTable, specifiedColor=<rgba 0 0 0 0>}
       cornerRadii=<4 4 4 4>
      }
      tintColor {colorUseRule=UseColor_Specified, specifiedColor=<rgba 0.0048 0.0048 0.0048 1>}
      uVRegion {bIsValid=false, max=<0 0>, min=<0 0>}
     }
     foregroundColor {colorUseRule=UseColor_Specified, specifiedColor=<rgba 1 1 1 1>}
     hoveredBackgroundBrush {
      bIsDynamicallyLoaded=false
      drawAs=RoundedBox
      imageType=NoImage
      mirroring=NoMirror
      resourceName=None
      resourceObject=null
      tiling=NoTile
      imageSize=<0 0>
      margin=<ltrb 0 0 0 0>
      outlineSettings {
       bUseBrushTransparency=false
       roundingType=FixedRadius
       width=1
       color {colorUseRule=UseColor_ColorTable, specifiedColor=<rgba 0 0 0 0>}
       cornerRadii=<4 4 4 4>
      }
      tintColor {colorUseRule=UseColor_Specified, specifiedColor=<rgba 0.0048 0.0048 0.0048 1>}
      uVRegion {bIsValid=false, max=<0 0>, min=<0 0>}
     }
     hoveredFillBrush {
      bIsDynamicallyLoaded=false
      drawAs=RoundedBox
      imageType=NoImage
      mirroring=NoMirror
      resourceName=None
      resourceObject=null
      tiling=NoTile
      imageSize=<0 0>
      margin=<ltrb 0 0 0 0>
      outlineSettings {
       bUseBrushTransparency=false
       roundingType=FixedRadius
       width=1
       color {colorUseRule=UseColor_Specified, specifiedColor=<rgba 0 0 0 0>}
       cornerRadii=<4 4 4 4>
      }
      tintColor {colorUseRule=UseColor_Specified, specifiedColor=<rgba 0.0953 0.0953 0.0953 1>}
      uVRegion {bIsValid=false, max=<0 0>, min=<0 0>}
     }
     inactiveFillBrush {
      bIsDynamicallyLoaded=false
      drawAs=RoundedBox
      imageType=NoImage
      mirroring=NoMirror
      resourceName=None
      resourceObject=null
      tiling=NoTile
      imageSize=<0 0>
      margin=<ltrb 0 0 0 0>
      outlineSettings {
       bUseBrushTransparency=false
       roundingType=FixedRadius
       width=1
       color {colorUseRule=UseColor_Specified, specifiedColor=<rgba 0 0 0 0>}
       cornerRadii=<4 4 4 4>
      }
      tintColor {colorUseRule=UseColor_Specified, specifiedColor=<rgba 0.0395 0.0395 0.0395 1>}
      uVRegion {bIsValid=false, max=<0 0>, min=<0 0>}
     }
     insetPadding=<ltrb 1 1 1 1>
     textPadding=<ltrb 7 2.5 7 2.5>
    }
   }
   slot {class=HorizontalBoxSlot, properties {verticalAlignment=VAlign_Center}}
  }
  txtLabel_1 {
   widgetClassPath=&CommonTextBlock
   parent=HorizontalBox
   widget=txtLabel_1
   properties {
    text=Y:
    font {
     bForceMonospaced=false
     bMaterialIsStencil=false
     fontMaterial=null
     fontObject=&NotoSans
     letterSpacing=0
     monospacedWidth=1
     size=12
     skewAmount=0
     typefaceFontName=NotoSans-Regular
     outlineSettings {
      bApplyOutlineToDropShadows=false
      bMiteredCorners=false
      bSeparateFillAlpha=false
      outlineMaterial=null
      outlineSize=0
      outlineColor=<rgba 0 0 0 1>
     }
    }
    shadowOffset=<0 0>
   }
   slot {class=HorizontalBoxSlot, properties {verticalAlignment=VAlign_Center}}
  }
  SpinBox_y {
   widgetClassPath=&SpinBox
   bIsVariable=true
   parent=HorizontalBox
   widget=SpinBox_y
   properties {
    foregroundColor {colorUseRule=UseColor_Specified, specifiedColor=<rgba 1 1 1 1>}
    widgetStyle {
     activeBackgroundBrush {
      bIsDynamicallyLoaded=false
      drawAs=RoundedBox
      imageType=NoImage
      mirroring=NoMirror
      resourceName=None
      resourceObject=null
      tiling=NoTile
      imageSize=<0 0>
      margin=<ltrb 0 0 0 0>
      outlineSettings {
       bUseBrushTransparency=false
       roundingType=FixedRadius
       width=1
       color {colorUseRule=UseColor_ColorTable, specifiedColor=<rgba 0 0 0 0>}
       cornerRadii=<4 4 4 4>
      }
      tintColor {colorUseRule=UseColor_Specified, specifiedColor=<rgba 0.0048 0.0048 0.0048 1>}
      uVRegion {bIsValid=false, max=<0 0>, min=<0 0>}
     }
     activeFillBrush {
      bIsDynamicallyLoaded=false
      drawAs=RoundedBox
      imageType=NoImage
      mirroring=NoMirror
      resourceName=None
      resourceObject=null
      tiling=NoTile
      imageSize=<0 0>
      margin=<ltrb 0 0 0 0>
      outlineSettings {
       bUseBrushTransparency=false
       roundingType=FixedRadius
       width=1
       color {colorUseRule=UseColor_Specified, specifiedColor=<rgba 0 0 0 0>}
       cornerRadii=<4 4 4 4>
      }
      tintColor {colorUseRule=UseColor_Specified, specifiedColor=<rgba 0 0.162 0.7454 1>}
      uVRegion {bIsValid=false, max=<0 0>, min=<0 0>}
     }
     arrowsImage {
      bIsDynamicallyLoaded=false
      drawAs=NoDrawType
      imageType=NoImage
      mirroring=NoMirror
      resourceName=None
      resourceObject=null
      tiling=NoTile
      imageSize=<0 0>
      margin=<ltrb 0 0 0 0>
      outlineSettings {
       bUseBrushTransparency=false
       roundingType=HalfHeightRadius
       width=0
       color {colorUseRule=UseColor_Specified, specifiedColor=<rgba 0 0 0 0>}
       cornerRadii=<0 0 0 0>
      }
      tintColor {colorUseRule=UseColor_Specified, specifiedColor=<rgba 1 1 1 1>}
      uVRegion {bIsValid=false, max=<0 0>, min=<0 0>}
     }
     backgroundBrush {
      bIsDynamicallyLoaded=false
      drawAs=RoundedBox
      imageType=NoImage
      mirroring=NoMirror
      resourceName=None
      resourceObject=null
      tiling=NoTile
      imageSize=<0 0>
      margin=<ltrb 0 0 0 0>
      outlineSettings {
       bUseBrushTransparency=false
       roundingType=FixedRadius
       width=1
       color {colorUseRule=UseColor_ColorTable, specifiedColor=<rgba 0 0 0 0>}
       cornerRadii=<4 4 4 4>
      }
      tintColor {colorUseRule=UseColor_Specified, specifiedColor=<rgba 0.0048 0.0048 0.0048 1>}
      uVRegion {bIsValid=false, max=<0 0>, min=<0 0>}
     }
     foregroundColor {colorUseRule=UseColor_Specified, specifiedColor=<rgba 1 1 1 1>}
     hoveredBackgroundBrush {
      bIsDynamicallyLoaded=false
      drawAs=RoundedBox
      imageType=NoImage
      mirroring=NoMirror
      resourceName=None
      resourceObject=null
      tiling=NoTile
      imageSize=<0 0>
      margin=<ltrb 0 0 0 0>
      outlineSettings {
       bUseBrushTransparency=false
       roundingType=FixedRadius
       width=1
       color {colorUseRule=UseColor_ColorTable, specifiedColor=<rgba 0 0 0 0>}
       cornerRadii=<4 4 4 4>
      }
      tintColor {colorUseRule=UseColor_Specified, specifiedColor=<rgba 0.0048 0.0048 0.0048 1>}
      uVRegion {bIsValid=false, max=<0 0>, min=<0 0>}
     }
     hoveredFillBrush {
      bIsDynamicallyLoaded=false
      drawAs=RoundedBox
      imageType=NoImage
      mirroring=NoMirror
      resourceName=None
      resourceObject=null
      tiling=NoTile
      imageSize=<0 0>
      margin=<ltrb 0 0 0 0>
      outlineSettings {
       bUseBrushTransparency=false
       roundingType=FixedRadius
       width=1
       color {colorUseRule=UseColor_Specified, specifiedColor=<rgba 0 0 0 0>}
       cornerRadii=<4 4 4 4>
      }
      tintColor {colorUseRule=UseColor_Specified, specifiedColor=<rgba 0.0953 0.0953 0.0953 1>}
      uVRegion {bIsValid=false, max=<0 0>, min=<0 0>}
     }
     inactiveFillBrush {
      bIsDynamicallyLoaded=false
      drawAs=RoundedBox
      imageType=NoImage
      mirroring=NoMirror
      resourceName=None
      resourceObject=null
      tiling=NoTile
      imageSize=<0 0>
      margin=<ltrb 0 0 0 0>
      outlineSettings {
       bUseBrushTransparency=false
       roundingType=FixedRadius
       width=1
       color {colorUseRule=UseColor_Specified, specifiedColor=<rgba 0 0 0 0>}
       cornerRadii=<4 4 4 4>
      }
      tintColor {colorUseRule=UseColor_Specified, specifiedColor=<rgba 0.0395 0.0395 0.0395 1>}
      uVRegion {bIsValid=false, max=<0 0>, min=<0 0>}
     }
     insetPadding=<ltrb 1 1 1 1>
     textPadding=<ltrb 7 2.5 7 2.5>
    }
   }
   slot {class=HorizontalBoxSlot, properties {verticalAlignment=VAlign_Center}}
  }
  txtLabel_2 {
   widgetClassPath=&CommonTextBlock
   parent=HorizontalBox
   widget=txtLabel_2
   properties {
    text=Z:
    font {
     bForceMonospaced=false
     bMaterialIsStencil=false
     fontMaterial=null
     fontObject=&NotoSans
     letterSpacing=0
     monospacedWidth=1
     size=12
     skewAmount=0
     typefaceFontName=NotoSans-Regular
     outlineSettings {
      bApplyOutlineToDropShadows=false
      bMiteredCorners=false
      bSeparateFillAlpha=false
      outlineMaterial=null
      outlineSize=0
      outlineColor=<rgba 0 0 0 1>
     }
    }
    shadowOffset=<0 0>
   }
   slot {class=HorizontalBoxSlot, properties {verticalAlignment=VAlign_Center}}
  }
  SpinBox_Z {
   widgetClassPath=&SpinBox
   bIsVariable=true
   parent=HorizontalBox
   widget=SpinBox_Z
   properties {
    foregroundColor {colorUseRule=UseColor_Specified, specifiedColor=<rgba 1 1 1 1>}
    widgetStyle {
     activeBackgroundBrush {
      bIsDynamicallyLoaded=false
      drawAs=RoundedBox
      imageType=NoImage
      mirroring=NoMirror
      resourceName=None
      resourceObject=null
      tiling=NoTile
      imageSize=<0 0>
      margin=<ltrb 0 0 0 0>
      outlineSettings {
       bUseBrushTransparency=false
       roundingType=FixedRadius
       width=1
       color {colorUseRule=UseColor_ColorTable, specifiedColor=<rgba 0 0 0 0>}
       cornerRadii=<4 4 4 4>
      }
      tintColor {colorUseRule=UseColor_Specified, specifiedColor=<rgba 0.0048 0.0048 0.0048 1>}
      uVRegion {bIsValid=false, max=<0 0>, min=<0 0>}
     }
     activeFillBrush {
      bIsDynamicallyLoaded=false
      drawAs=RoundedBox
      imageType=NoImage
      mirroring=NoMirror
      resourceName=None
      resourceObject=null
      tiling=NoTile
      imageSize=<0 0>
      margin=<ltrb 0 0 0 0>
      outlineSettings {
       bUseBrushTransparency=false
       roundingType=FixedRadius
       width=1
       color {colorUseRule=UseColor_Specified, specifiedColor=<rgba 0 0 0 0>}
       cornerRadii=<4 4 4 4>
      }
      tintColor {colorUseRule=UseColor_Specified, specifiedColor=<rgba 0 0.162 0.7454 1>}
      uVRegion {bIsValid=false, max=<0 0>, min=<0 0>}
     }
     arrowsImage {
      bIsDynamicallyLoaded=false
      drawAs=NoDrawType
      imageType=NoImage
      mirroring=NoMirror
      resourceName=None
      resourceObject=null
      tiling=NoTile
      imageSize=<0 0>
      margin=<ltrb 0 0 0 0>
      outlineSettings {
       bUseBrushTransparency=false
       roundingType=HalfHeightRadius
       width=0
       color {colorUseRule=UseColor_Specified, specifiedColor=<rgba 0 0 0 0>}
       cornerRadii=<0 0 0 0>
      }
      tintColor {colorUseRule=UseColor_Specified, specifiedColor=<rgba 1 1 1 1>}
      uVRegion {bIsValid=false, max=<0 0>, min=<0 0>}
     }
     backgroundBrush {
      bIsDynamicallyLoaded=false
      drawAs=RoundedBox
      imageType=NoImage
      mirroring=NoMirror
      resourceName=None
      resourceObject=null
      tiling=NoTile
      imageSize=<0 0>
      margin=<ltrb 0 0 0 0>
      outlineSettings {
       bUseBrushTransparency=false
       roundingType=FixedRadius
       width=1
       color {colorUseRule=UseColor_ColorTable, specifiedColor=<rgba 0 0 0 0>}
       cornerRadii=<4 4 4 4>
      }
      tintColor {colorUseRule=UseColor_Specified, specifiedColor=<rgba 0.0048 0.0048 0.0048 1>}
      uVRegion {bIsValid=false, max=<0 0>, min=<0 0>}
     }
     foregroundColor {colorUseRule=UseColor_Specified, specifiedColor=<rgba 1 1 1 1>}
     hoveredBackgroundBrush {
      bIsDynamicallyLoaded=false
      drawAs=RoundedBox
      imageType=NoImage
      mirroring=NoMirror
      resourceName=None
      resourceObject=null
      tiling=NoTile
      imageSize=<0 0>
      margin=<ltrb 0 0 0 0>
      outlineSettings {
       bUseBrushTransparency=false
       roundingType=FixedRadius
       width=1
       color {colorUseRule=UseColor_ColorTable, specifiedColor=<rgba 0 0 0 0>}
       cornerRadii=<4 4 4 4>
      }
      tintColor {colorUseRule=UseColor_Specified, specifiedColor=<rgba 0.0048 0.0048 0.0048 1>}
      uVRegion {bIsValid=false, max=<0 0>, min=<0 0>}
     }
     hoveredFillBrush {
      bIsDynamicallyLoaded=false
      drawAs=RoundedBox
      imageType=NoImage
      mirroring=NoMirror
      resourceName=None
      resourceObject=null
      tiling=NoTile
      imageSize=<0 0>
      margin=<ltrb 0 0 0 0>
      outlineSettings {
       bUseBrushTransparency=false
       roundingType=FixedRadius
       width=1
       color {colorUseRule=UseColor_Specified, specifiedColor=<rgba 0 0 0 0>}
       cornerRadii=<4 4 4 4>
      }
      tintColor {colorUseRule=UseColor_Specified, specifiedColor=<rgba 0.0953 0.0953 0.0953 1>}
      uVRegion {bIsValid=false, max=<0 0>, min=<0 0>}
     }
     inactiveFillBrush {
      bIsDynamicallyLoaded=false
      drawAs=RoundedBox
      imageType=NoImage
      mirroring=NoMirror
      resourceName=None
      resourceObject=null
      tiling=NoTile
      imageSize=<0 0>
      margin=<ltrb 0 0 0 0>
      outlineSettings {
       bUseBrushTransparency=false
       roundingType=FixedRadius
       width=1
       color {colorUseRule=UseColor_Specified, specifiedColor=<rgba 0 0 0 0>}
       cornerRadii=<4 4 4 4>
      }
      tintColor {colorUseRule=UseColor_Specified, specifiedColor=<rgba 0.0395 0.0395 0.0395 1>}
      uVRegion {bIsValid=false, max=<0 0>, min=<0 0>}
     }
     insetPadding=<ltrb 1 1 1 1>
     textPadding=<ltrb 7 2.5 7 2.5>
    }
   }
   slot {class=HorizontalBoxSlot, properties {verticalAlignment=VAlign_Center}}
  }
  detailOffsetWorld {
   widgetClassPath=&labeledEntry_Widget
   bIsVariable=true
   parent=HorizontalBox
   widget=detailOffsetWorld
   properties {defaultBool=true, label=World Space:}
   slot {class=HorizontalBoxSlot, properties {horizontalAlignment=HAlign_Center, verticalAlignment=VAlign_Center, padding=<ltrb 5 0 0 0>}}
  }
 ]
}
