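# Editor utility widget for tech art that embeds the ActorLayout_EWidget panel above a "Get Distance To Selected Actor" button.
# On click, it takes the first selected actor, if any, and measures the distance from the level viewport camera to that actor's location.
# It prints that distance in units with the actor's display name as an orange on-screen message.
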EditorUtilityWidgetBlueprint /Game/Blueprints/Blutilities/ForTechArt/AssetTools/GeneralAssetTools/AssetRefedByBR_EWidget
graphs [EventGraph {nodes [
   "On Clicked (TA_Generic01_Button)"!() -> (OutputDelegate, then, widget:User Widget Object Reference, event:Pointer Event Structure) {is=title K2Node_ComponentBoundEvent}
   > n1 = GetSelectedActors()
   > Branch(Condition=Is Valid Index!(TargetArray:Array of Actor Object References=@n1, IndexToTest:int) -> (ReturnValue:bool))
   n2 = "Get (a copy)"(Array:Array of Actor Object References=@n1) -> (Output:Actor Object Reference)
   n3 = Get UnrealEditorSubsystem()
   n4 = Vector Length(A=vector - vector!(
     A:Vector=Get Actor Location(self=@n2)
     B:Vector=GetLevelViewportCameraInfo!(self:Unreal Editor Subsystem Object Reference=@n3) -> (CameraLocation:Vector, CameraRotation:Rotator, ReturnValue:bool)
    ) -> (ReturnValue:Vector) {is=title K2Node_CallFunction})
   > PrintString(
    InString=Append!(A="To String (Float)"(InDouble=@n4), B=" units from  ", C:String=GetDisplayName(Object=@n2)) -> (ReturnValue)
    TextColor=(R=1,G=0.4941,B=0.124,A=1)
    Duration=5
   )
  ]}]
variables []
widgets {
 info {rootWidgetClass=&VerticalBox, widgetCount=3}
 widgets [
  VerticalBox_0 {widgetClassPath=&VerticalBox, widget=VerticalBox_0}
  ActorLayout_EWidget {
   widgetClassPath=&/ActorLayout_EWidget
   bIsVariable=true
   parent=VerticalBox_0
   widget=ActorLayout_EWidget
   slot {class=VerticalBoxSlot}
  }
  TA_Generic01_Button {
   widgetClassPath=&TA_Generic01_Button
   parent=VerticalBox_0
   widget=TA_Generic01_Button
   properties {label=Get Distance To Selected Actor}
   slot {class=VerticalBoxSlot, properties {padding=<ltrb 0 30 0 10>}}
  }
 ]
}
