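# Editor utility widget for tech art that finds selected textures whose compression keeps alpha but whose alpha channel appears blank.
# Each candidate is drawn through an inspection material into a render target, downscaled by the Accuracy setting, then read pixel by pixel for non-uniform values.
# Blank results fill the left list and the local AATexturesWithBadAlpha collection.
# Items can swap between two lists, be selected in the Content Browser or opened; Debug shows the render target.
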EditorUtilityWidgetBlueprint /Game/Blueprints/Blutilities/ForTechArt/AssetTools/TextureTools/TextureAlphaInspector_EWidget
graphs [
 EventGraph {nodes [
   n1 = GetAssetsInCollection!(
    execute
    self:Asset Tags Subsystem Object Reference=Get AssetTagsSubsystem() -> (ReturnValue:Asset Tags Subsystem Object Reference) {is=title K2Node_GetEngineSubsystem}
    Name:Name=None
   ) -> (then, ReturnValue:Array of Asset Data Structures)
   > RunAssetsThroughFilter!(execute:Exec, self:Object Reference, AssetDataList:Array of Asset Data Structures=@n1, Filter:"ARFilter Structure (by ref)") -> (then:Exec) {is=title K2Node_Message}
   "On Clicked (btnLoadTextures)"!() -> (OutputDelegate, then, widget:User Widget Object Reference, event:Pointer Event Structure) {is=title K2Node_ComponentBoundEvent}
   > Load Textures from Selection Validated()
   n2 = "On Selection Changed (cbDebug)"!() -> (OutputDelegate, then, selected:bool, widget:User Widget Object Reference) {is=title K2Node_ComponentBoundEvent}
   > SetVisibility(self=Get imgDebug() -> (imgDebug:Image Object Reference), InVisibility=Select(Option 0=Collapsed, Option 1=Visible, Index=@n2.selected))
   n3 = "On Item Left Clicked (list1)"!() -> (OutputDelegate, then, isShiftDown:bool, isCtrlDown:bool, data:Object Reference, item:User Widget Object Reference) {is=title K2Node_ComponentBoundEvent}
   > Branch(Condition=@n3.isShiftDown)
   > n4 = Cast To TA_GenericListItemData!(execute, Object=@n3.data) -> (then, CastFailed, AsTA Generic List Item Data:TA Generic List Item Data Object Reference) {is=title K2Node_DynamicCast}
   n5 = "Get (a copy)"(Array:Array of Object References=Get objects!(self:TA Generic List Item Data Object Reference=@n4) -> (objects:Array of Object References)) -> (Output:Object Reference)
   > SyncBrowserToObjects(AssetPaths=Make Array("[0]":String=Get Object Path String(Object=@n5)) -> (Array:Array of Strings))
   n6 = "On Item Right Clicked (list1)"!() -> (OutputDelegate, then, isShiftDown:bool, isCtrlDown:bool, data:Object Reference, item:User Widget Object Reference) {is=title K2Node_ComponentBoundEvent}
   > n7 = Cast To TA_GenericListItemData(Object=@n6.data)
   > OpenEditorForAssets!(
    execute
    self:Asset Editor Subsystem Object Reference=Get AssetEditorSubsystem()
    Assets:Array of Object References=Get objects(self=@n7)
    OpenedMethod:EAssetTypeActivationOpenedMethod Enum=Edit
   ) -> (then, ReturnValue:bool)
   "On Clicked (TA_Generic01_Button)"!() -> (OutputDelegate, then, widget:User Widget Object Reference, event:Pointer Event Structure) {is=title K2Node_ComponentBoundEvent}
   > swap()
   "On Item Selection Changed (list1)"!() -> (
    OutputDelegate
    then
    selectedData:Array of Object References
    selectedWidgets:Array of User Widget Object References
    selectedIndices:Array of Integers
   ) {is=title K2Node_ComponentBoundEvent}
   n8 = Get list2() -> (list2:TA Generic 01 List Object Reference)
   > n9 = getSelectedItems!(execute, self:TA Generic 01 List Object Reference=@n8) -> (then, isAnySelected:bool, items:Array of User Widget Object References, selectedData:Array of Object References)
   > Branch(Condition=@n9)
   > clearSelection(self:TA Generic 01 List Object Reference=@n8)
   "On Item Selection Changed (list2)"!() -> (
    OutputDelegate
    then
    selectedData:Array of Object References
    selectedWidgets:Array of User Widget Object References
    selectedIndices:Array of Integers
   ) {is=title K2Node_ComponentBoundEvent}
   n10 = Get list1() -> (list1:TA Generic 01 List Object Reference)
   > n11 = getSelectedItems(self=@n10)
   > Branch(Condition=@n11)
   > clearSelection(self=@n10)
   "On Clicked (btnClearRight)"!() -> (OutputDelegate, then, widget:User Widget Object Reference, event:Pointer Event Structure) {is=title K2Node_ComponentBoundEvent}
   > clearItems(self:TA Generic 01 List Object Reference=Get list2())
   "On Clicked (btnSelectRight)"!() -> (OutputDelegate, then, widget:User Widget Object Reference, event:Pointer Event Structure) {is=title K2Node_ComponentBoundEvent}
   > selectFromList!(execute, self, list:TA Generic 01 List Object Reference=Get list2()) -> (then)
   "On Clicked (btnSelectLeft)"!() -> (OutputDelegate, then, widget:User Widget Object Reference, event:Pointer Event Structure) {is=title K2Node_ComponentBoundEvent}
   > selectFromList(list=Get list1())
   Event Destruct()
   > Parent: Destruct()
   > Set currentRT(currentRT:Texture Render Target 2D Object Reference) -> (Output_Get:Texture Render Target 2D Object Reference)
   Event Tick()
   n12 = Get debugTestRT() -> (debugTestRT:Texture Object Reference)
   > Is Valid(InputObject=@n12)
   > Branch(Condition="Not Equal (Object)"!(A:Object Reference=@n12, B:Object Reference=Get lastDebugTestRT() -> (lastDebugTestRT:Texture Object Reference)) -> (ReturnValue:bool))
   > Set lastDebugTestRT(lastDebugTestRT:Texture Object Reference=Get debugTestRT()) -> (Output_Get:Texture Object Reference)
   > n13 = GetDynamicMaterial(self=Get imgDebug())
   > SetTextureParameterValue(self=@n13, ParameterName=Texture, Value=Get debugTestRT())
   > SetVisibility(self=Get imgDebug(), InVisibility=Visible)
   Set isMipped(isMipped:bool=NOT Boolean(A=Get isMipped() -> (isMipped:bool))) -> (Output_Get:bool)
   > n14 = For Each Loop(Array:Array of Material Instance Dynamic Object References=Get currentMaterials() -> (currentMaterials:Array of Material Instance Dynamic Object References)) -> (Array Element:Material Instance Dynamic Object Reference)
   > SetScalarParameterValue(self=@n14, ParameterName=mipbias, Value=Select(Option 0:double=0, Option 1:double=20, Index=Get isMipped()) -> (ReturnValue:double))
   GetAsset(InAssetData="Get (a copy)"(Array:Array of Asset Data Structures=@n1) -> (Output:Asset Data Structure))
  ]}
 Load Textures from Selection Validated {nodes [
   Load Textures from Selection Validated() {is=title K2Node_FunctionEntry}
   > n1 = Sequence()
   > Set currentRT()
   > Clear(TargetArray:Array of Texture 2D Object References=Get currentTextures() -> (currentTextures:Array of Texture 2D Object References))
   > Clear(TargetArray:Array of Material Instance Dynamic Object References=Get currentMaterials())
   > clearItems(self=Get list1())
   > n2 = GetSelectedAssets()
   > n3 = Filter by Class!(
    execute
    TargetArray:Array of Object References=@n2
    ObjectClass:Object Class Reference=/Script/Engine.Texture2D
    FilterType:EEditorScriptingFilterType Enum
   ) -> (then, ReturnValue:Array of Texture 2D Object References)
   > n4 = For Each Loop(Array:Array of Texture 2D Object References=@n3) -> (Array Element:Texture 2D Object Reference)
   n5 = Get CompressionNoAlpha!(self:Texture Object Reference=@n4) -> (CompressionNoAlpha:bool)
   n6 = Get CompressionSettings(self=@n4)
   > Branch(Condition=AND Boolean!(
     A=NOT Boolean(A=@n5)
     B="Not Equal (Enum)"!(A:TextureCompressionSettings Enum=@n6, B:TextureCompressionSettings Enum=TC_HDR) -> (ReturnValue:bool)
     C:bool="Not Equal (Enum)"(A=@n6, B=TC_Grayscale)
     D:bool="Not Equal (Enum)"(A=@n6, B=TC_Displacementmap)
     E:bool="Not Equal (Enum)"(A=@n6, B=TC_DistanceFieldFont)
     F:bool="Not Equal (Enum)"(A=@n6, B=TC_HalfFloat)
     G:bool="Not Equal (Enum)"(A=@n6, B=TC_Normalmap)
     H:bool="Not Equal (Enum)"(A=@n6, B=TC_HDR_Compressed)
     I:bool="Not Equal (Enum)"(A=@n6, B=TC_VectorDisplacementmap)
    ) -> (ReturnValue))
   > n7 = Run Alpha Test on Texture!(execute, self, textureIn:Texture 2D Object Reference=@n4) -> (then, notEmpty:bool)
   > n8 = Branch(Condition=@n7)
   addTextureToList!(execute=@n8.else, self, Texture:Texture 2D Object Reference=@n4) -> (then)
   n9 = MakeLiteralName(Value=AATexturesWithBadAlpha)
   n10 = CollectionExists!(execute=@n1.then_1, self:Asset Tags Subsystem Object Reference=Get AssetTagsSubsystem(), Name:Name=@n9) -> (then, ReturnValue:bool)
   > n11 = Branch(Condition=@n10)
   > n12 = EmptyCollection!(execute, self:Asset Tags Subsystem Object Reference=Get AssetTagsSubsystem(), Name:Name=@n9) -> (then, ReturnValue:bool)
   AddAssetPtrsToCollection!(
    execute=[@n13.then, @n12.then]
    self:Asset Tags Subsystem Object Reference=Get AssetTagsSubsystem()
    Name:Name=@n9
    AssetPtrs:Array of Object References=Get currentTextures()
   ) -> (then, ReturnValue:bool)
   n13 = CreateCollection!(
    execute=@n11.else
    self:Asset Tags Subsystem Object Reference=Get AssetTagsSubsystem()
    Name:Name=@n9
    ShareType:ECollectionScriptingShareType Enum=Local
   ) -> (then, ReturnValue:bool)
   n14 = Get CompressionSettings()
   "Equal (Enum)"(A:TextureCompressionSettings Enum=@n14, B:TextureCompressionSettings Enum=TC_Default)
   "Equal (Enum)"(A=@n14, B=TC_Default)
   "Equal (Enum)"(A=@n14, B=TC_Masks)
   "Equal (Enum)"(A=@n14, B=TC_HDR)
   "Equal (Enum)"(A=@n14, B=TC_EditorIcon)
   "Equal (Enum)"(A=@n14, B=TC_BC7)
   "Equal (Enum)"(A=@n14, B=TC_VectorDisplacementmap)
  ]}
 LoadTexturesFromSelection {nodes [
   LoadTexturesFromSelection() {is=title K2Node_FunctionEntry}
   > Clear(TargetArray:Array of Texture 2D Object References=Get currentTextures())
   > Clear(TargetArray:Array of Material Instance Dynamic Object References=Get currentMaterials())
   > clearItems(self=Get list1())
   > n1 = GetSelectedAssets()
   > n2 = Filter by Class(TargetArray=@n1, ObjectClass=&Texture2D)
   > n3 = For Each Loop(Array=@n2)
   n4 = Get CompressionNoAlpha(self=@n3)
   n5 = Get CompressionSettings(self=@n3)
   n6 = "Equal (Enum)"(A=@n5, B=TC_Masks)
   n7 = "Equal (Enum)"(A=@n5, B=TC_Default)
   > Branch(Condition=AND Boolean!(A=NOT Boolean(A=@n4), B=OR Boolean(A=@n7, B=@n6)) -> (ReturnValue))
   > addTextureToList(Texture=@n3)
  ]}
 Run Alpha Test on Texture {nodes [
   n1 = Run Alpha Test on Texture!() -> (then, textureIn:Texture 2D Object Reference) {is=title K2Node_FunctionEntry}
   > Set l_TextureIn(l_TextureIn:Texture 2D Object Reference=@n1) -> (Output_Get:Texture 2D Object Reference)
   n2 = Get l_TextureIn() -> (l_TextureIn:Texture 2D Object Reference)
   > Set l_sizeX(l_sizeX:double="To Float (Integer)"(InInt=GetSizeX!(self:Texture 2D Object Reference=@n2) -> (ReturnValue:int=0))) -> (Output_Get:double)
   > Set l_sizeY(l_sizeY:double="To Float (Integer)"(InInt=GetSizeY!(self:Texture 2D Object Reference=@n2) -> (ReturnValue:int=0))) -> (Output_Get:double)
   n3 = Get spinAccuracy() -> (spinAccuracy:Spin Box Object Reference)
   > Set l_minSize(l_minSize:double="Min (Float)"!(
     A:double=Get l_sizeY() -> (l_sizeY:double)
     B:double=Get l_sizeX() -> (l_sizeX:double)
     C:double=GetValue!(self:Spin Box Object Reference=@n3) -> (ReturnValue:float)
    ) -> (ReturnValue:double) {is=title K2Node_CommutativeAssociativeBinaryOperator}) -> (Output_Get:double)
   > n4 = Branch(Condition="float > float"(A=Get l_sizeX(), B=Get l_sizeY()) {is=title K2Node_CallFunction})
   n5 = float / float(A=Get l_sizeX(), B=Get l_sizeY()) {is=title K2Node_CallFunction}
   n6 = Get l_minSize() -> (l_minSize:double)
   > n7 = Set l_targetRes(l_targetRes:Vector 2D Structure=MakeVector2D(X=float * float(A=@n5, B=@n6), Y=Get l_minSize())) -> (Output_Get:Vector 2D Structure)
   n8 = Get!() -> (l_targetRes_X:double, l_targetRes_Y:double)
   n9 = Truncate(A=@n8)
   n10 = Truncate(A=@n8.l_targetRes_Y)
   n11 = writeTextureToRenderTarget!(execute=[@n15.then, @n7.then], self, Width:int=@n9, Height:int=@n10, TextureIn:Texture Object Reference=Get l_TextureIn()) -> (then, outRenderTarget:Texture Render Target 2D Object Reference)
   > n12 = validateAlpha!(
    execute
    self
    renderTargetIn:Texture Render Target 2D Object Reference=@n11
    size_X:double="To Float (Integer)"(InInt=@n9)
    size_Y:double="To Float (Integer)"(InInt=@n10)
   ) -> (then, notEmpty:bool)
   > Return Node!(execute, notEmpty:bool=@n12)
   n13 = float / float(A=Get l_sizeY(), B=Get l_sizeX())
   n14 = Get l_minSize()
   n15 = Set l_targetRes(execute=@n4.else, l_targetRes=MakeVector2D(X=Get l_minSize(), Y=float * float(A=@n13, B=@n14)))
  ]}
 addTextureToList {nodes [
   n1 = addTextureToList!() -> (then, Texture:Texture 2D Object Reference) {is=title K2Node_FunctionEntry}
   > Add(TargetArray:Array of Texture 2D Object References=Get currentTextures(), NewItem:"Texture 2D Object Reference (by ref)"=@n1) {is=title K2Node_CallArrayFunction}
   > n2 = Sequence()
   > n3 = CreateDynamicMaterialInstance(
    WorldContextObject=Self-Reference()
    Parent=/Game/Blueprints/Blutilities/ForTechArt/AssetTools/TextureTools/Materials/TextureAlphaInspectionMIPPED_UIMat
    OptionalName=None
    CreationFlags=None
   )
   > Add(
    TargetArray:Array of Material Instance Dynamic Object References=Get currentMaterials()
    NewItem:"Material Instance Dynamic Object Reference (by ref)"=@n3
   )
   > SetTextureParameterValue(self=@n3, ParameterName=Texture, Value=Get Texture() -> (Texture:Texture 2D Object Reference))
   n4 = Construct Object from Class(
    execute=@n2.then_1
    Class=/Game/Blueprints/Blutilities/ForTechArt/Data/TA_GenericListItemData.TA_GenericListItemData_C
    self=Self-Reference()
   ) -> (ReturnValue:TA Generic List Item Data Object Reference)
   n5 = Get currentMaterials()
   n6 = "Get (a copy)"(
    Array:Array of Material Instance Dynamic Object References=@n5
    Dimension 1=Last Index!(TargetArray:Array of Material Instance Dynamic Object References=@n5) -> (ReturnValue=0)
   ) -> (Output:Material Instance Dynamic Object Reference)
   > Set displayBrush!(
    execute
    displayBrush:Slate Brush Structure=Make Slate Brush!(
     TintColor:Slate Color Structure=(SpecifiedColor=(R=1,G=1,B=1,A=1),ColorUseRule=UseColor_Specified)
     DrawAs:ESlateBrushDrawType Enum=Image
     Tiling:ESlateBrushTileType Enum=NoTile
     Mirroring:ESlateBrushMirrorType Enum=NoMirror
     ImageSize:Vector 2D Structure=(X=128,Y=128)
     Margin:Margin Structure=(Left=0,Top=0,Right=0,Bottom=0)
     ResourceObject:Object Reference=@n6
     OutlineSettings:Slate Brush Outline Settings Structure=(CornerRadii=(X=0,Y=0,Z=0,W=0),Color=(SpecifiedColor=(R=0,G=0,B=0,A=0),ColorUseRule=UseColor_Specified),Width=0,RoundingType=HalfHeightRadius,bUseBrushTransparency=False)
    ) -> (SlateBrush:Slate Brush Structure) {is=title K2Node_MakeStruct}
    self:TA Generic List Item Data Object Reference=@n4
   ) -> (then, Output_Get:Slate Brush Structure)
   n7 = Get Texture()
   n8 = GetSizeX(self=@n7) -> (ReturnValue=0)
   n9 = GetSizeY(self=@n7) -> (ReturnValue=0)
   > Set displayString!(
    execute
    displayString:String=Append!(A="To String (Integer)"(InInt=@n8), B=x, C:String="To String (Integer)"(InInt=@n9), D:String=" ::: ", E:String=GetDisplayName(Object=@n7)) -> (ReturnValue)
    self:TA Generic List Item Data Object Reference=@n4
   ) -> (then, Output_Get:String)
   > Set objects!(
    execute
    objects:Array of Object References=Make Array!("[0]":Object Reference=Get Texture(), "[1]":Object Reference=@n6) -> (Array:Array of Object References)
    self:TA Generic List Item Data Object Reference=@n4
   ) -> (then, Output_Get:Array of Object References)
   > addItem!(execute, self:TA Generic 01 List Object Reference=Get list1(), dataObject:Object Reference=@n4) -> (then, newItem:User Widget Object Reference, newData:Object Reference)
   MakeBrushFromTexture!(Texture:Texture 2D Object Reference=Get Texture(), Width:int=128, Height:int=128) -> (ReturnValue:Slate Brush Structure)
   CreateAssetData(bAllowBlueprintClass=false)
  ]}
 getRenderTarget {nodes [
   n1 = getRenderTarget!() -> (then, Width:int, Height:int) {is=title K2Node_FunctionEntry}
   n2 = CreateRenderTarget2D!(
    execute=[@n1.then, @n4.Is Not Valid]
    WorldContextObject:Object Reference
    Width:int=@n1
    Height:int=@n1.Height
    Format:ETextureRenderTargetFormat Enum=RTF_RGBA8
    ClearColor:Linear Color Structure=(R=0,G=0,B=0,A=1)
    bAutoGenerateMipMaps:bool=false
    bSupportUAVs:bool=false
   ) -> (then, ReturnValue:Texture Render Target 2D Object Reference)
   > n3 = Set currentRT(currentRT=@n2)
   Return Node!(
    execute=[@n4.Is Valid, @n3.then]
    ReturnValue:Texture Render Target 2D Object Reference=Get currentRT() -> (currentRT:Texture Render Target 2D Object Reference)
   )
   n4 = Is Valid(InputObject=Get currentRT())
   Get RenderTargetFormat!(self:Texture Render Target 2D Object Reference=@n2) -> (RenderTargetFormat:ETextureRenderTargetFormat Enum=RTF_R8)
  ]}
 selectFromList {nodes [
   n1 = selectFromList!() -> (then, list:TA Generic 01 List Object Reference) {is=title K2Node_FunctionEntry}
   > n2 = getListData!(execute, self:TA Generic 01 List Object Reference=@n1) -> (then, data:Array of Object References)
   > n3 = For Each Loop(Array:Array of Object References=@n2) -> (Array Element:Object Reference)
   SyncBrowserToObjects(execute=@n3.Completed, AssetPaths=Get l_selectList() -> (l_selectList:Array of Strings))
   n4 = Cast To TA_GenericListItemData(execute=@n3.LoopBody, Object=@n3)
   n5 = Get objects(self=@n4)
   > Add(
    TargetArray:Array of Strings=Get l_selectList()
    NewItem:"String (by ref)"=Get Object Path String(Object="Get (a copy)"(Array:Array of Object References=@n5) -> (Output:Object Reference))
   )
  ]}
 setTestMat {nodes [setTestMat!() -> (then, Value:Texture Object Reference) {is=title K2Node_FunctionEntry}]}
 swap {nodes [
   swap() {is=title K2Node_FunctionEntry}
   > n1 = Sequence()
   > n2 = getSelectedItems(self=Get list1())
   > n3 = Branch(Condition=@n2)
   > Set l_listBeingCopiedFrom(l_listBeingCopiedFrom:TA Generic 01 List Object Reference=Get list1()) -> (Output_Get:TA Generic 01 List Object Reference)
   > Set l_listTarget(l_listTarget:TA Generic 01 List Object Reference=Get list2()) -> (Output_Get:TA Generic 01 List Object Reference)
   Set l_listBeingCopiedFrom(execute=@n3.else, l_listBeingCopiedFrom=Get list2())
   > Set l_listTarget(l_listTarget=Get list1())
   n4 = getSelectedItems(execute=@n1.then_1, self=Get l_listBeingCopiedFrom() -> (l_listBeingCopiedFrom:TA Generic 01 List Object Reference))
   > Branch(Condition=@n4)
   > addItems!(
    execute
    self:TA Generic 01 List Object Reference=Get l_listTarget() -> (l_listTarget:TA Generic 01 List Object Reference)
    dataObjects:Array of Object References=@n4.selectedData
   ) -> (then)
   > removeItems!(execute, self:TA Generic 01 List Object Reference=Get l_listBeingCopiedFrom(), itemWidgets:Array of User Widget Object References=@n4.items) -> (then)
  ]}
 validateAlpha {nodes [
   n1 = validateAlpha!() -> (then, renderTargetIn:Texture Render Target 2D Object Reference, size:Vector 2D Structure) {is=title K2Node_FunctionEntry}
   n2 = BreakVector2D(InVec=@n1.size)
   > Set l_sizeX(l_sizeX=@n2)
   > Set l_sizeY(l_sizeY=@n2.Y)
   > Set l_total(l_total:double=float * float(A=Get l_sizeX(), B=Get l_sizeY())) -> (Output_Get:double)
   n3 = Get l_total() -> (l_total:double)
   > n4 = For Loop with Break!(Execute:Exec, FirstIndex:int, LastIndex:int=Truncate(A=float - float(A=@n3, B=1) {is=title K2Node_CallFunction}), Break:Exec=@n16.then) -> (LoopBody:Exec, Index:int, Completed:Exec) {is=title K2Node_MacroInstance}
   Return Node!(execute=@n4.Completed, notEmpty:bool=Get l_notEmpty() -> (l_notEmpty:bool))
   n5 = "To Float (Integer)"(InInt=@n4)
   n6 = Get l_sizeX()
   n7 = "To Float (Integer)"(InInt=@n4)
   n8 = Get l_sizeX()
   > n9 = ReadRenderTargetPixel!(
    execute
    WorldContextObject:Object Reference
    TextureRenderTarget:Texture Render Target 2D Object Reference=Get renderTargetIn() -> (renderTargetIn:Texture Render Target 2D Object Reference)
    X:int=Truncate(A="% (Float)"!(A:double=@n5, B:double=@n6) -> (ReturnValue:double))
    Y:int=Floor(A=float / float(A=@n7, B=@n8)) {is=title K2Node_CallFunction}
   ) -> (then, ReturnValue_B:Byte, ReturnValue_G:Byte, ReturnValue_R:Byte, ReturnValue_A:Byte)
   n10 = "Equal (Byte)"!(A:Byte=@n9.ReturnValue_R, B:Byte=0) -> (ReturnValue:bool)
   n11 = "Equal (Byte)"(A=@n9.ReturnValue_R, B=255)
   n12 = Get l_firstValue() -> (l_firstValue:int=0)
   n13 = "Equal (Byte)"(A=@n9.ReturnValue_A, B="To Byte (Integer)"!(InInt:int=@n12) -> (ReturnValue:Byte=0))
   n14 = "Not Equal (Integer)"!(A:int=@n12, B:int=-1) -> (ReturnValue:bool) {is=title K2Node_CallFunction}
   > n15 = Branch(Condition=AND Boolean(A=OR Boolean(A=@n10, B=@n11), B=Select(Option 0:bool=true, Option 1:bool=@n13, Index=@n14) -> (ReturnValue:bool)))
   > Set l_firstValue(l_firstValue:int="To Integer (Byte)"!(InByte:Byte=@n9.ReturnValue_R) -> (ReturnValue:int=0)) -> (Output_Get:int=0)
   n16 = Set l_notEmpty(execute=@n15.else, l_notEmpty:bool=true) -> (Output_Get:bool)
   n17 = Branch(Condition=Get l_notEmpty())
   > PrintString(InString=Not Empty, TextColor=(R=0,G=1,B=0.3366,A=1))
   PrintString(execute=@n17.else, InString=" Empty", TextColor=(R=1,G=0,B=0.0083,A=1))
  ]}
 writeTextureToRenderTarget {nodes [
   n1 = BeginDrawCanvasToRenderTarget!(execute, WorldContextObject:Object Reference, TextureRenderTarget:Texture Render Target 2D Object Reference) -> (then, Canvas:Canvas Object Reference, Size:Vector 2D Structure, Context:Draw To Render Target Context Structure)
   > Draw Texture!(
    execute
    self:Canvas Object Reference=@n1
    RenderTexture:Texture Object Reference=Get TextureIn() -> (TextureIn:Texture Object Reference)
    ScreenPosition:Vector 2D Structure
    ScreenSize:Vector 2D Structure=@n1.Size
    CoordinatePosition:Vector 2D Structure
    CoordinateSize:Vector 2D Structure=(X=1,Y=1)
    RenderColor:Linear Color Structure=(R=1,G=1,B=1,A=1)
    BlendMode:EBlendMode Enum=BLEND_Opaque
    Rotation:float=0
    PivotPoint:Vector 2D Structure=(X=0.5,Y=0.5)
   ) -> (then)
   n2 = writeTextureToRenderTarget!() -> (then, Width:int, Height:int, TextureIn:Texture Object Reference) {is=title K2Node_FunctionEntry}
   > n3 = getRenderTarget!(execute, self, Width:int=@n2, Height:int=@n2.Height) -> (then, ReturnValue:Texture Render Target 2D Object Reference) {is=title K2Node_CallFunction}
   > Set l_RT(l_RT:Texture Render Target 2D Object Reference=@n3) -> (Output_Get:Texture Render Target 2D Object Reference)
   > n4 = CreateDynamicMaterialInstance(
    Parent=/Game/Blueprints/Blutilities/ForTechArt/AssetTools/TextureTools/Materials/TextureAlphaInspection_UIMat
    OptionalName=None
    CreationFlags=None
   )
   > SetTextureParameterValue(self=@n4, ParameterName=Texture, Value=Get TextureIn())
   > DrawMaterialToRenderTarget!(
    execute
    WorldContextObject:Object Reference
    TextureRenderTarget:Texture Render Target 2D Object Reference=Get l_RT() -> (l_RT:Texture Render Target 2D Object Reference)
    Material:Material Interface Object Reference=@n4
   ) -> (then)
   > n5 = GetDynamicMaterial(self=Get imgDebug())
   > SetTextureParameterValue(self=@n5, ParameterName=Texture, Value=Get l_RT())
   n6 = MakeVector2D(X="To Float (Integer)"(InInt=Get Width() -> (Width:int=0)), Y="To Float (Integer)"(InInt=Get Height() -> (Height:int=0)))
   > SetDesiredSizeOverride!(
    execute
    self:Image Object Reference=Get imgDebug()
    DesiredSize:Vector 2D Structure=vector2d * float!(A:Vector 2D Structure=Normalize 2D!(A:Vector 2D Structure=@n6) -> (ReturnValue:Vector 2D Structure), B:double=128) -> (ReturnValue:Vector 2D Structure)
   ) -> (then)
   > Return Node!(execute, outRenderTarget:Texture Render Target 2D Object Reference=Get l_RT())
  ]}
]
variables [
 currentMaterials {type="array<MaterialInstanceDynamic>", default []}
 currentRT {type=TextureRenderTarget2D, default=null}
 currentTextures {type="array<Texture2D>", default []}
 debugTestRT {type=Texture, default=null}
 isMipped {type=boolean}
 lastDebugTestRT {type=Texture, default=null}
]
widgets {
 info {rootWidgetClass=&VerticalBox, widgetCount=18}
 widgets [
  VerticalBox_0 {widgetClassPath=&VerticalBox, widget=VerticalBox_0}
  HorizontalBox_4 {widgetClassPath=&HorizontalBox, parent=VerticalBox_0, widget=HorizontalBox_4, slot {class=VerticalBoxSlot}}
  btnLoadTextures {
   widgetClassPath=&TA_Generic01_Button
   parent=HorizontalBox_4
   widget=btnLoadTextures
   properties {label=Load Selected Textures with Alpha}
   slot {class=HorizontalBoxSlot, properties {verticalAlignment=VAlign_Center}}
  }
  CommonTextBlock_63 {
   widgetClassPath=&CommonTextBlock
   parent=HorizontalBox_4
   widget=CommonTextBlock_63
   properties {
    text=Accuracy:
    font {
     bForceMonospaced=false
     bMaterialIsStencil=false
     fontMaterial=null
     fontObject=&NotoSans
     letterSpacing=0
     monospacedWidth=1
     size=14
     skewAmount=0
     typefaceFontName=NotoSans-Regular
     outlineSettings {
      bApplyOutlineToDropShadows=false
      bMiteredCorners=false
      bSeparateFillAlpha=false
      outlineMaterial=null
      outlineSize=0
      outlineColor=<rgba 0 0 0 1>
     }
    }
    shadowOffset=<0 0>
   }
   slot {class=HorizontalBoxSlot, properties {verticalAlignment=VAlign_Center, padding=<ltrb 10 0 0 0>}}
  }
  SizeBox_0 {
   widgetClassPath=&SizeBox
   parent=HorizontalBox_4
   widget=SizeBox_0
   properties {bOverride_WidthOverride=true, widthOverride=58.7652}
   slot {class=HorizontalBoxSlot, properties {verticalAlignment=VAlign_Center}}
  }
  spinAccuracy {
   widgetClassPath=/Script/UMG.SpinBox
   bIsVariable=true
   parent=SizeBox_0
   widget=spinAccuracy
   properties {
    bOverride_MaxSliderValue=true
    bOverride_MaxValue=true
    bOverride_MinSliderValue=true
    bOverride_MinValue=true
    maxSliderValue=256
    maxValue=256
    minSliderValue=8
    minValue=8
    toolTipText="To reduce the number of pixels we have to iterate we reduce the texture dimensions to this size when checking.  The lower the number, the faster the check will be."
    value=24
    foregroundColor {colorUseRule=UseColor_Specified, specifiedColor=<rgba 1 1 1 1>}
    widgetStyle {
     activeBackgroundBrush {
      bIsDynamicallyLoaded=false
      drawAs=RoundedBox
      imageType=NoImage
      mirroring=NoMirror
      resourceName=None
      resourceObject=null
      tiling=NoTile
      imageSize=<0 0>
      margin=<ltrb 0 0 0 0>
      outlineSettings {
       bUseBrushTransparency=false
       roundingType=FixedRadius
       width=1
       color {colorUseRule=UseColor_ColorTable, specifiedColor=<rgba 0 0 0 0>}
       cornerRadii=<4 4 4 4>
      }
      tintColor {colorUseRule=UseColor_Specified, specifiedColor=<rgba 0.0048 0.0048 0.0048 1>}
      uVRegion {bIsValid=false, max=<0 0>, min=<0 0>}
     }
     activeFillBrush {
      bIsDynamicallyLoaded=false
      drawAs=RoundedBox
      imageType=NoImage
      mirroring=NoMirror
      resourceName=None
      resourceObject=null
      tiling=NoTile
      imageSize=<0 0>
      margin=<ltrb 0 0 0 0>
      outlineSettings {
       bUseBrushTransparency=false
       roundingType=FixedRadius
       width=1
       color {colorUseRule=UseColor_Specified, specifiedColor=<rgba 0 0 0 0>}
       cornerRadii=<4 4 4 4>
      }
      tintColor {colorUseRule=UseColor_Specified, specifiedColor=<rgba 0 0.162 0.7454 1>}
      uVRegion {bIsValid=false, max=<0 0>, min=<0 0>}
     }
     arrowsImage {
      bIsDynamicallyLoaded=false
      drawAs=NoDrawType
      imageType=NoImage
      mirroring=NoMirror
      resourceName=None
      resourceObject=null
      tiling=NoTile
      imageSize=<0 0>
      margin=<ltrb 0 0 0 0>
      outlineSettings {
       bUseBrushTransparency=false
       roundingType=HalfHeightRadius
       width=0
       color {colorUseRule=UseColor_Specified, specifiedColor=<rgba 0 0 0 0>}
       cornerRadii=<0 0 0 0>
      }
      tintColor {colorUseRule=UseColor_Specified, specifiedColor=<rgba 1 1 1 1>}
      uVRegion {bIsValid=false, max=<0 0>, min=<0 0>}
     }
     backgroundBrush {
      bIsDynamicallyLoaded=false
      drawAs=RoundedBox
      imageType=NoImage
      mirroring=NoMirror
      resourceName=None
      resourceObject=null
      tiling=NoTile
      imageSize=<0 0>
      margin=<ltrb 0 0 0 0>
      outlineSettings {
       bUseBrushTransparency=false
       roundingType=FixedRadius
       width=1
       color {colorUseRule=UseColor_ColorTable, specifiedColor=<rgba 0 0 0 0>}
       cornerRadii=<4 4 4 4>
      }
      tintColor {colorUseRule=UseColor_Specified, specifiedColor=<rgba 0.0048 0.0048 0.0048 1>}
      uVRegion {bIsValid=false, max=<0 0>, min=<0 0>}
     }
     foregroundColor {colorUseRule=UseColor_Specified, specifiedColor=<rgba 1 1 1 1>}
     hoveredBackgroundBrush {
      bIsDynamicallyLoaded=false
      drawAs=RoundedBox
      imageType=NoImage
      mirroring=NoMirror
      resourceName=None
      resourceObject=null
      tiling=NoTile
      imageSize=<0 0>
      margin=<ltrb 0 0 0 0>
      outlineSettings {
       bUseBrushTransparency=false
       roundingType=FixedRadius
       width=1
       color {colorUseRule=UseColor_ColorTable, specifiedColor=<rgba 0 0 0 0>}
       cornerRadii=<4 4 4 4>
      }
      tintColor {colorUseRule=UseColor_Specified, specifiedColor=<rgba 0.0048 0.0048 0.0048 1>}
      uVRegion {bIsValid=false, max=<0 0>, min=<0 0>}
     }
     hoveredFillBrush {
      bIsDynamicallyLoaded=false
      drawAs=RoundedBox
      imageType=NoImage
      mirroring=NoMirror
      resourceName=None
      resourceObject=null
      tiling=NoTile
      imageSize=<0 0>
      margin=<ltrb 0 0 0 0>
      outlineSettings {
       bUseBrushTransparency=false
       roundingType=FixedRadius
       width=1
       color {colorUseRule=UseColor_Specified, specifiedColor=<rgba 0 0 0 0>}
       cornerRadii=<4 4 4 4>
      }
      tintColor {colorUseRule=UseColor_Specified, specifiedColor=<rgba 0.0953 0.0953 0.0953 1>}
      uVRegion {bIsValid=false, max=<0 0>, min=<0 0>}
     }
     inactiveFillBrush {
      bIsDynamicallyLoaded=false
      drawAs=RoundedBox
      imageType=NoImage
      mirroring=NoMirror
      resourceName=None
      resourceObject=null
      tiling=NoTile
      imageSize=<0 0>
      margin=<ltrb 0 0 0 0>
      outlineSettings {
       bUseBrushTransparency=false
       roundingType=FixedRadius
       width=1
       color {colorUseRule=UseColor_Specified, specifiedColor=<rgba 0 0 0 0>}
       cornerRadii=<4 4 4 4>
      }
      tintColor {colorUseRule=UseColor_Specified, specifiedColor=<rgba 0.0395 0.0395 0.0395 1>}
      uVRegion {bIsValid=false, max=<0 0>, min=<0 0>}
     }
     insetPadding=<ltrb 1 1 1 1>
     textPadding=<ltrb 7 2.5 7 2.5>
    }
   }
   slot {class=SizeBoxSlot}
  }
  cbDebug {
   widgetClassPath=/Game/Blueprints/Blutilities/ForTechArt/Widgets/Buttons/TA_Generic01_Checkbox
   bIsVariable=true
   parent=HorizontalBox_4
   widget=cbDebug
   properties {label=Debug}
   slot {class=HorizontalBoxSlot, properties {horizontalAlignment=HAlign_Right, verticalAlignment=VAlign_Center, size {sizeRule=Fill, value=1}}}
  }
  HorizontalBox_0 {
   widgetClassPath=&HorizontalBox
   parent=VerticalBox_0
   widget=HorizontalBox_0
   slot {class=VerticalBoxSlot, properties {size {sizeRule=Fill, value=1}}}
  }
  list1 {
   widgetClassPath=/Game/Blueprints/Blutilities/ForTechArt/Widgets/List/TA_Generic01_List
   bIsVariable=true
   parent=HorizontalBox_0
   widget=list1
   slot {class=HorizontalBoxSlot, properties {size {sizeRule=Fill, value=1}}}
  }
  TA_Generic01_Button {
   widgetClassPath=&TA_Generic01_Button
   parent=HorizontalBox_0
   widget=TA_Generic01_Button
   properties {label="< - SWAP ->"}
   slot {class=HorizontalBoxSlot, properties {verticalAlignment=VAlign_Center, padding=<ltrb 10 10 10 10>}}
  }
  list2 {
   widgetClassPath=&TA_Generic01_List
   bIsVariable=true
   parent=HorizontalBox_0
   widget=list2
   slot {class=HorizontalBoxSlot, properties {size {sizeRule=Fill, value=1}}}
  }
  HorizontalBox_1 {
   widgetClassPath=&HorizontalBox
   parent=VerticalBox_0
   widget=HorizontalBox_1
   slot {class=VerticalBoxSlot, properties {padding=<ltrb 0 20 0 0>}}
  }
  Spacer_101 {
   widgetClassPath=&Spacer
   parent=HorizontalBox_1
   widget=Spacer_101
   slot {class=HorizontalBoxSlot, properties {size {sizeRule=Fill, value=1}}}
  }
  btnClearRight {
   widgetClassPath=&TA_Generic01_Button
   parent=HorizontalBox_1
   widget=btnClearRight
   properties {label=Clear Right List}
   slot {class=HorizontalBoxSlot, properties {padding=<ltrb 90 0 30 0>, size {sizeRule=Fill, value=1}}}
  }
  HorizontalBox {
   widgetClassPath=&HorizontalBox
   parent=VerticalBox_0
   widget=HorizontalBox
   slot {class=VerticalBoxSlot, properties {padding=<ltrb 0 10 0 0>}}
  }
  btnSelectLeft {
   widgetClassPath=&TA_Generic01_Button
   parent=HorizontalBox
   widget=btnSelectLeft
   properties {label=Select Assets In Left List}
   slot {class=HorizontalBoxSlot, properties {padding=<ltrb 30 0 30 0>, size {sizeRule=Fill, value=1}}}
  }
  btnSelectRight {
   widgetClassPath=&TA_Generic01_Button
   parent=HorizontalBox
   widget=btnSelectRight
   properties {label=Select Assets In Right List}
   slot {class=HorizontalBoxSlot, properties {padding=<ltrb 30 0 30 0>, size {sizeRule=Fill, value=1}}}
  }
  imgDebug {
   widgetClassPath=&Image
   bIsVariable=true
   parent=VerticalBox_0
   widget=imgDebug
   properties {
    visibility=Collapsed
    brush {
     bIsDynamicallyLoaded=false
     drawAs=Image
     imageType=NoImage
     mirroring=NoMirror
     resourceName=None
     resourceObject=/Game/Blueprints/Blutilities/ForTechArt/AssetTools/TextureTools/Materials/TextureDisplay_UIMat
     tiling=NoTile
     imageSize=<256 256>
     margin=<ltrb 0 0 0 0>
     outlineSettings {
      bUseBrushTransparency=false
      roundingType=HalfHeightRadius
      width=0
      color {colorUseRule=UseColor_Specified, specifiedColor=<rgba 0 0 0 0>}
      cornerRadii=<0 0 0 0>
     }
     tintColor {colorUseRule=UseColor_Specified, specifiedColor=<rgba 1 1 1 1>}
     uVRegion {bIsValid=false, max=<0 0>, min=<0 0>}
    }
   }
   slot {class=VerticalBoxSlot, properties {horizontalAlignment=HAlign_Center, verticalAlignment=VAlign_Center}}
  }
 ]
}
