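# Editor utility widget that hosts asset tools as tabs in one panel, with a horizontal tab list above an animated widget switcher.
# On Construct, Initialize Tabs adds Asset Validator, Positioning Tools and Find Actors By Property, then selects the first tab.
# addNewTab creates each tab item and its tool widget; list selection switches pages by matching index.
# The Texture Alpha Checker tab call has no incoming execution, so it never runs.
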EditorUtilityWidgetBlueprint /Game/Blueprints/Blutilities/ForTechArt/AssetTools/GeneralAssetTools/GeneralAssetTools_EWidget
graphs [
 EventGraph {nodes [
   n1 = "On Item Selection Changed (listTabs)"!() -> (
    OutputDelegate
    then
    selectedData:Array of Object References
    selectedWidgets:Array of User Widget Object References
    selectedIndices:Array of Integers
   ) {is=title K2Node_ComponentBoundEvent}
   > Branch(Condition=Is Valid Index!(TargetArray:Array of Object References=@n1.selectedData, IndexToTest:int) -> (ReturnValue:bool))
   > n2 = getSelectedIndex!(execute, self:TA Generic 01 List Object Reference=Get listTabs() -> (listTabs:TA Generic 01 List Object Reference)) -> (then, index:int=0)
   > SetActiveWidgetIndex(self=Get widgetSwitcher() -> (widgetSwitcher:Common Widget Switcher Legacy Object Reference), Index=@n2)
   Event Construct()
   > Initialize Tabs()
  ]}
 Initialize Tabs {nodes [
   n1 = addNewTab!(
    execute
    self
    label:String=Texture Alpha Checker
    menu:User Widget Class Reference=/Game/Blueprints/Blutilities/ForTechArt/AssetTools/TextureTools/TextureAlphaInspector_EWidget.TextureAlphaInspector_EWidget_C
   ) -> (then)
   addNewTab(
    execute=[@n3.then, @n1.then]
    label=Find Actors By Property
    menu=/Game/Blueprints/Blutilities/ForLDs/FindByPropertyTool/FindByProperty_EWidget.FindByProperty_EWidget_C
   )
   Initialize Tabs() {is=title K2Node_FunctionEntry}
   > n2 = Sequence()
   > addNewTab(
    label=Asset Validator
    menu=/Game/Blueprints/Blutilities/ForArtists/AssetValidationTool/AssetValidationTool_EWidget.AssetValidationTool_EWidget_C
   )
   > n3 = addNewTab(
    label=Positioning Tools
    menu=&/AssetRefedByBR_EWidget.AssetRefedByBR_EWidget_C
   )
   setIndexSelected!(execute=@n2.then_1, self:TA Generic 01 List Object Reference=Get listTabs(), index:int=0, selected:bool=true, throwEvent:bool=true) -> (then)
  ]}
 addNewTab {nodes [
   n1 = addNewTab!() -> (then, label:String, menu:User Widget Class Reference) {is=title K2Node_FunctionEntry}
   > addItemGeneric!(execute, self:TA Generic 01 List Object Reference=Get listTabs(), displayString:String=@n1) -> (then, widget:User Widget Object Reference, data:Object Reference)
   > n2 = Create Widget!(
    execute:Exec
    WorldContextObject:Object Reference=Self-Reference()
    Class:User Widget Class Reference=@n1.menu
    OwningPlayer:Player Controller Object Reference
   ) -> (then:Exec, ReturnValue:User Widget Object Reference)
   > AddChild(self=Get widgetSwitcher(), Content=@n2)
  ]}
]
variables []
widgets {
 info {rootWidgetClass=&VerticalBox, widgetCount=5}
 widgets [
  VerticalBox_20 {widgetClassPath=&VerticalBox, widget=VerticalBox_20}
  listTabs {
   widgetClassPath=/Game/Blueprints/Blutilities/ForTechArt/Widgets/List/TA_Generic01_List
   bIsVariable=true
   parent=VerticalBox_20
   widget=listTabs
   properties {
    listItemClass=/Game/Blueprints/Blutilities/ForTechArt/Widgets/Buttons/TA_Generic01_TabItem.TA_Generic01_TabItem_C
    scrollOrientation=Orient_Horizontal
   }
   slot {class=VerticalBoxSlot}
  }
  Image_109 {
   widgetClassPath=&Image
   bIsVariable=true
   parent=VerticalBox_20
   widget=Image_109
   properties {
    brush {
     bIsDynamicallyLoaded=false
     drawAs=Image
     imageType=NoImage
     mirroring=NoMirror
     resourceName=None
     resourceObject=null
     tiling=NoTile
     imageSize=<32 2>
     margin=<ltrb 0 0 0 0>
     outlineSettings {
      bUseBrushTransparency=false
      roundingType=HalfHeightRadius
      width=0
      color {colorUseRule=UseColor_Specified, specifiedColor=<rgba 0 0 0 0>}
      cornerRadii=<0 0 0 0>
     }
     tintColor {colorUseRule=UseColor_Specified, specifiedColor=<rgba 1 1 1 1>}
     uVRegion {bIsValid=false, max=<0 0>, min=<0 0>}
    }
    colorAndOpacity=<rgba 0.0208 0.0208 0.0208 1>
   }
   slot {class=VerticalBoxSlot}
  }
  Border_0 {
   widgetClassPath=&Border
   parent=VerticalBox_20
   widget=Border_0
   properties {brushColor=<rgba 0 0 0 0.1>, padding=<ltrb 10 10 10 10>}
   slot {class=VerticalBoxSlot, properties {size {sizeRule=Fill, value=1}}}
  }
  widgetSwitcher {
   widgetClassPath=&CommonWidgetSwitcherLegacy
   bIsVariable=true
   parent=Border_0
   widget=widgetSwitcher
   properties {transitionCurveType=QuadIn, transitionDuration=0.2, transitionType=Horizontal}
   slot {class=BorderSlot, properties {padding=<ltrb 10 10 10 10>}}
  }
 ]
}
